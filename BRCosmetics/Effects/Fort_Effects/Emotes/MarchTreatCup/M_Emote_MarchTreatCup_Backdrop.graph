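# Translucent, unlit, two-sided Niagara particle material for the MarchTreatCup emote backdrop, with three looks blended by the Use_Line and Use_PointLine scalars.
# The base look colors two rotated, panning cell-noise layers; Line masks opacity with scrolling streaks.
# PointLine combines floored-sine stripes, gradient-tinted, with scrolling pixels multiplied by a rainbow texture.
# Particle color drives all looks, while vertical gradients, view-angle fade and DepthFade soften opacity.
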Material /BRCosmetics/Effects/Fort_Effects/Emotes/MarchTreatCup/M_Emote_MarchTreatCup_Backdrop
graph {
 properties {
  bAutomaticallySetUsageInEditor=false
  bUsedWithNiagaraMeshParticles=true
  bUsedWithNiagaraSprites=true
  blendMode=BLEND_Translucent
  shadingModel=MSM_Unlit
  twoSided=true
  usedShadingModels=Unlit
 }
 nodes [
  n1 = ParticleColor()
  n2 = TextureCoordinate() {vTiling=2}
  n3 = Constant2Vector() {g=0.2}
  n4 = Power(Base=ComponentMask(None=Add(A=@n2, B=@n3)) {g=true}, Exp=Constant() {r=1.5}) {constExponent=4}
  n5 = TextureCoordinate() {vTiling=2.5}
  n6 = Panner(Coordinate=Power(Base=ComponentMask(None=@n5) {g=true}, Exp=Constant() {r=1.5}) {constExponent=4}) {speedY=1.5}
  n7 = Multiply(A=ComponentMask(None=TextureCoordinate()) {r=true}) {constB=10}
  n8 = TextureCoordinate()
  n9 = Multiply(A=OneMinus(None=ComponentMask(None=@n8) {r=true})) {constB=10}
  n10 = Multiply(A=Saturate(None=Power(Base=@n7) {constExponent=3}), B=Saturate(None=Power(Base=@n9) {constExponent=3}))
  n11 = Add(
   A=TextureSample(UVs=Panner(Coordinate=@n4) {speedY=1.5}) {texture=/Game/Weapons/FORT_Rifles/Effects/Textures/T_FX_ThinLine_BoomB_01}.R
   B=Multiply(A=TextureSample(UVs=@n6) {texture=&/T_Shuriken_Trail_NoClamp}.R, B=Saturate(None=@n10))
  )
  n12 = ScalarParameter() {parameterName=Use_Line}
  n13 = Power(Base=ComponentMask(None=TextureCoordinate()) {g=true}) {constExponent=4}
  n14 = Multiply(A=Time(), B=Constant() {r=0.5})
  n15 = Add(A=Sine(None=Add(A=@n13, B=@n14)) {period=0.2}) {constB=0.03}
  n16 = TextureCoordinate()
  n17 = Power(Base=Multiply(A=ComponentMask(None=@n16) {g=true}) {constB=1.5}) {constExponent=4}
  n18 = Multiply(A=Multiply(A=Floor(None=@n15), B=Saturate(None=@n17)))
  n19 = Sine(None=ComponentMask(None=TextureCoordinate() {uTiling=60}) {r=true})
  n20 = Multiply(A=Floor(None=Add(A=@n19) {constB=0.05}), B=ComponentMask(None=TextureCoordinate()) {g=true})
  n21 = TextureSample(UVs=Panner(Coordinate=TextureCoordinate() {uTiling=2}) {speedY=1}) {texture=/Game/Effects/Fort_Effects/Emotes/Candor/T_Noise_Blur_Texture}
  n22 = Add(
   A=Multiply(A=@n18, B=ScalarParameter() {defaultValue=1, parameterName=BoxLine_Mul}) {constB=0}
   B=Multiply(A=Multiply(A=@n20), B=LinearInterpolate(Alpha=@n21.R))
  ) {constB=0}
  n23 = VectorParameter() {parameterName=PointTexContral, defaultValue=<rgba 1 1.2 0 1>}
  n24 = Multiply(A=TextureCoordinate(), B=AppendVector(A=@n23.R, B=@n23.R))
  n25 = ComponentMask(None=@n24) {r=true}
  n26 = Power(Base=ComponentMask(None=@n24) {g=true})
  n27 = TextureSample(UVs=Panner(Coordinate=AppendVector(A=@n25, B=@n26), Time=Time()) {speedY=1}) {
   samplerType=SAMPLERTYPE_LinearGrayscale
   texture=/Game/UI/Frontend/ItemShop/Com/vfx/T_CommunityVoting_Activities_Pixels.CommunityVoting_Activities_Pixels
  }
  n28 = TextureCoordinate() {uTiling=3, vTiling=0}
  n29 = Multiply(A=@n27.RGB, B=TextureSample(UVs=Panner(Coordinate=@n28) {speedY=1}) {texture=&T_Tile_RainbowColor}.RGB)
  n30 = ScalarParameter() {parameterName=Use_PointLine}
  n31 = LinearInterpolate(A=LinearInterpolate(A=@n1.A, B=Multiply(A=@n1.A, B=@n11), Alpha=@n12), B=Multiply(A=@n1.A, B=Add(A=@n22, B=@n29)), Alpha=@n30)
  n32 = ComponentMask(None=TextureCoordinate()) {g=true}
  n33 = Saturate(None=Power(Base=Multiply(A=@n32) {constB=1.5}) {constExponent=1.5})
  n34 = OneMinus(None=ComponentMask(None=TextureCoordinate()) {g=true})
  n35 = Saturate(None=Power(Base=Multiply(A=@n34) {constB=5}) {constExponent=5})
  n36 = LinearInterpolate(A=Saturate(None=@n33), B=Saturate(None=Multiply(A=@n33, B=@n35)), Alpha=ScalarParameter() {parameterName=Use_Line})
  n37 = Multiply(A=MaterialFunctionCall("FadeSensitivity (optional)"=ScalarParameter() {parameterName=MeshFade}) {materialFunction=&FadeBasedOnViewAngle}.Result) {constB=3}
  DepthFade(Opacity=Multiply(A=Multiply(A=@n31, B=@n36), B=Clamp(None=@n37)), FadeDistance=ScalarParameter() {parameterName=DepthFade})
  n38 = VectorParameter() {parameterName=DropTex01_Contral, defaultValue=<rgba 0 1 -0.6 0.3>}
  n39 = Multiply(A=TextureCoordinate(), B=AppendVector(A=@n38.R, B=@n38.R))
  n40 = ScalarParameter() {defaultValue=0.25, parameterName=DropTexRotator}
  n41 = TextureSampleParameter2D!(
   UVs=Panner(
    Coordinate=MaterialFunctionCall(UVs=@n39, "Rotation Angle (0-1)"=@n40) {materialFunction=&CustomRotator}.Rotated Values
    Speed=AppendVector(A=@n38.B, B=@n38.B)
   ) {speedX=-0.6, speedY=0.3}
   Level
   Apply View MipBias
  ) {
   mipValueMode=TMVM_MipLevel
   parameterName=DropTex01
   samplerType=SAMPLERTYPE_Grayscale
   texture=/BRCosmetics/Weapons/FORT_Melee/Pickaxe_Cube_Ninja_Male/FX/T_CellNoise
  }
  n42 = VectorParameter() {parameterName=DropTex02_Contral, defaultValue=<rgba 0 1 -0.5 0>}
  n43 = Multiply(A=TextureCoordinate(), B=AppendVector(A=@n42.R, B=@n42.R))
  n44 = TextureSampleParameter2D(UVs=Panner(
    Coordinate=MaterialFunctionCall(UVs=@n43, "Rotation Angle (0-1)"=@n40) {materialFunction=&CustomRotator}.Rotated Values
    Speed=AppendVector(A=@n42.B, B=@n42.B)
   ) {speedX=-0.5}) {
   mipValueMode=TMVM_MipLevel
   parameterName=DropTex02
   samplerType=SAMPLERTYPE_Grayscale
   texture=&T_CellNoise
  }
  n45 = Multiply(A=Add(A=@n41.R, B=@n44.R), B=@n1.RGB)
  n46 = Subtract(A=ComponentMask(None=TextureCoordinate()) {g=true}) {constB=0.3}
  n47 = LinearInterpolate(
   A=VectorParameter() {parameterName=PointLine_ColorA, defaultValue=<rgba 0 0.1379 0.4531 1>}.RGB
   B=VectorParameter() {parameterName=PointLine_ColorB, defaultValue=<rgba 0 1 0.5318 1>}.RGB
   Alpha=Saturate(None=Multiply(A=@n46) {constB=3})
  )
  n48 = VectorParameter() {parameterName=Square color, defaultValue=<rgba 4.1667 20 19.1929 0>}
  n49 = Desaturation(None=@n29, Fraction=Constant() {r=-5})
  n50 = Multiply(A=Add(A=Multiply(A=@n47, B=@n22), B=Multiply(A=@n48.RGB, B=@n49)))
  MaterialFunctionCall(Emissive=LinearInterpolate(A=LinearInterpolate(A=@n45, B=@n1.RGB, Alpha=@n12), B=Multiply(A=@n1.RGB, B=@n50), Alpha=@n30)) {materialFunction=&MF_FX_EmissiveAdjustments}
  n51 = TextureCoordinate()
  n52 = VectorParameter() {parameterName=Mosaic_Contral, defaultValue=<rgba 1 1 0 1>}
  n53 = AppendVector(A=@n52.R, B=@n52.R)
  n54 = TextureSampleParameter2D!(UVs=Panner(Coordinate=Multiply(A=@n51, B=@n53), Time=Time(), Speed=AppendVector(A=@n52.B, B=@n52.B)) {speedY=1}, Apply View MipBias) {parameterName=MosaicTex, samplerType=SAMPLERTYPE_LinearColor, texture=/Game/Global/Textures/Default/Noise/T_Atlas_DigitalNoise_Small}
  Multiply(A=@n18, B=LinearInterpolate(Alpha=@n54.B) {constA=-1})
 ]
}
