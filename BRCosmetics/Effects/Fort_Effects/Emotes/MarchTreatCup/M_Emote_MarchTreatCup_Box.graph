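# Unlit translucent Niagara sprite and mesh material for the box effect in the MarchTreatCup emote.
# Scalar lerps switch emissive between channel-masked box color, using a curve atlas sampled along instance bounds, and a sine-animated two-color edge line with panning noise.
# Opacity blends scanlined box alpha or edge line via Use_Scanline, then depth-fades.
# World position offset sums optional spline thickening and masked sine wobble, both driven by dynamic parameters.
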Material /BRCosmetics/Effects/Fort_Effects/Emotes/MarchTreatCup/M_Emote_MarchTreatCup_Box
graph {
 properties {
  bAutomaticallySetUsageInEditor=false
  bUsedWithNiagaraMeshParticles=true
  bUsedWithNiagaraSprites=true
  blendMode=BLEND_Translucent
  shadingModel=MSM_Unlit
  usedShadingModels=Unlit
 }
 nodes [
  n1 = Multiply(
   A=ScalarParameter() {defaultValue=1, parameterName=Spline_ThickMul}
   B=DynamicParameter() {parameterIndex=1, paramNames [Spline_Weight, Param2, Param3, Param4]}.Spline_Weight
  )
  n2 = MaterialFunctionCall(
   WidthBase=ScalarParameter() {defaultValue=3, parameterName=Spline_Base}
   WidthTip=ScalarParameter() {defaultValue=3, parameterName=Spline_Tip}
   Expand U or V UV Channel=StaticBool() {value=true}
  ) {materialFunction=&SplineThicken}
  n3 = Panner(Coordinate=TextureCoordinate())
  n4 = Sine(None=Multiply(A=Time(), B=DynamicParameter() {paramNames [SubUV_TilingU, SubUV_UVAdd, WPO_Random, Param4]}.WPO_Random))
  n5 = Multiply(
   A=Transform(None=Constant3Vector() {constant=<rgba 0 0 1 0>}) {transformSourceType=TRANSFORMSOURCE_Local}
   B=Multiply(
    A=TextureSample(UVs=@n3) {texture=&/T_Emote_MarchTreatCup_Mask}.RGB
    B=Clamp(None=@n4) {maxDefault=1.5}
   )
  )
  n6 = ScalarParameter() {defaultValue=1, parameterName=WPO_Intensity}
  Add(
   A=LinearInterpolate(A=Constant(), B=Multiply(A=@n1, B=@n2.WorldPositionOffset), Alpha=ScalarParameter() {parameterName=Use_SplineThicken})
   B=LinearInterpolate(B=Multiply(A=@n5, B=@n6), Alpha=ScalarParameter() {parameterName=Use_WPO})
  )
  n7 = ParticleColor()
  n8 = TextureSampleParameter2D() {parameterName=BaseTexture, texture=&/T_Emote_MarchTreatCup_Box}
  n9 = Multiply(A=@n7.A, B=LinearInterpolate(A=@n8.G, B=@n8.G, Alpha=ScalarParameter() {parameterName=Use_SecondSide}))
  n10 = MaterialFunctionCall!(
   Number of Scanlines=ScalarParameter() {defaultValue=180, parameterName=Scanline Amount}
   Scanline Speed
   "Scanlines Thickness (Bias)"=ScalarParameter() {defaultValue=-0.1, parameterName=ScanlineSize}
   Scanlines Dark Value=ScalarParameter() {defaultValue=0.5, parameterName=Scanlines Dark}
   Scanlines Light Value=ScalarParameter() {defaultValue=1, parameterName=Scanlines Light}
   Block Opacity
   UVs
   Time
   Block Repeat
  ) {materialFunction=/Game/Effects/Fort_Effects/Effects/Atlas/Materials/MF_Atlas_Scanlines_01}
  n11 = TextureSampleParameter2D(UVs=Panner(Coordinate=TextureCoordinate() {uTiling=0.2, vTiling=1.2}) {speedY=1}) {parameterName=EdgeLine_LerpTex, texture=/BRCosmetics/Gadgets/Assets/VinderTech_GliderChute/Glider_SlickSwish/FX/T_Glider_SlickSwish_noise}
  n12 = Multiply(
   A=TextureSampleParameter2D() {parameterName=EdgeLine_Tex, texture=&T_Emote_MarchTreatCup_Box}.R
   B=Power(Base=@n11.R) {constExponent=1}
  )
  DepthFade(
   Opacity=LinearInterpolate(A=Multiply(A=@n9, B=@n10.Result), B=Multiply(A=@n12, B=@n7.A), Alpha=ScalarParameter() {parameterName=Use_Scanline})
   FadeDistance=ScalarParameter() {defaultValue=100, parameterName=DepthFade}
  )
  n13 = Divide(A=MaterialFunctionCall() {materialFunction=/Game/Effects/Niagara/MasterMaterials/Functions/InstanceBounds_0-1_UVW}.B) {constB=3}
  n14 = ScalarParameter() {parameterName=ColorCurveOffset}
  n15 = CurveAtlasRowParameter(CurveTime=Multiply(A=Add(A=@n13, B=@n14) {constB=0.25})) {
   atlas=/BRCosmetics/Effects/Fort_Effects/Emotes/WarmShade/Materials/CAtlas_Emote_WarmShade
   curve=&/CV_Emote_MarchTreatCup_BackdropColor
   defaultValue=10
   parameterName=Color
  }
  n16 = ScalarParameter() {parameterName=Use_ColorCurve}
  n17 = Divide(A=MaterialFunctionCall() {materialFunction=&InstanceBounds_0-1_UVW}.B) {constB=3}
  n18 = ScalarParameter() {parameterName=ColorCurveOffset}
  n19 = CurveAtlasRowParameter(CurveTime=Multiply(A=Add(A=@n17, B=@n18) {constB=0.25})) {
   atlas=&CAtlas_Emote_WarmShade
   curve=&CV_Emote_MarchTreatCup_BackdropColor
   defaultValue=10
   parameterName=Color
  }
  n20 = LinearInterpolate(
   A=Multiply(A=LinearInterpolate(A=@n15, Alpha=@n16), B=@n8.G)
   B=Multiply(A=LinearInterpolate(A=@n19, Alpha=@n16), B=@n8.B)
   Alpha=ScalarParameter() {parameterName=Use_SecondSide}
  )
  n21 = VectorParameter() {parameterName=EdgeLine_ColorA, defaultValue=<rgba 1 0.1583 0.8098 1>}
  n22 = VectorParameter() {parameterName=EdgeLine_ColorB, defaultValue=<rgba 0.1 0.0911 0.651 1>}
  n23 = Multiply(A=ComponentMask(None=TextureCoordinate()) {g=true}, B=Constant() {r=1.5}) {constB=0.5}
  n24 = Multiply(A=Time(), B=Constant() {r=1}) {constB=3}
  n25 = Saturate(None=Sine(None=Add(A=@n23, B=@n24)))
  n26 = Multiply(A=Multiply(A=LinearInterpolate(A=@n21.RGB, B=@n22.RGB, Alpha=@n25)) {constB=5}, B=@n12)
  MaterialFunctionCall(Emissive=LinearInterpolate(A=Multiply(A=@n20, B=@n7.RGB), B=Multiply(A=@n26, B=@n7.RGB), Alpha=ScalarParameter() {parameterName=Use_EdgeLine})) {materialFunction=&MF_FX_EmissiveAdjustments}
 ]
}
