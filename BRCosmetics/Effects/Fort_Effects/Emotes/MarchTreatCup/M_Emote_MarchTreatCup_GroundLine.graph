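# Translucent unlit Niagara sprite and mesh particle material for the ground line effect of the MarchTreatCup emote.
# Emissive blends pink and blue through scrolling, mosaic-quantized noise, tinted by particle color, with an optional panning noise line added before MF_FX_EmissiveAdjustments.
# Opacity multiplies particle alpha by a sine-product UV falloff, with optional noise and world-space particle-centered sphere masks, then depth fade.
# Scalar parameters blend each optional layer in, avoiding static switches.
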Material /BRCosmetics/Effects/Fort_Effects/Emotes/MarchTreatCup/M_Emote_MarchTreatCup_GroundLine
graph {
 properties {
  bAutomaticallySetUsageInEditor=false
  bUsedWithNiagaraMeshParticles=true
  bUsedWithNiagaraSprites=true
  blendMode=BLEND_Translucent
  shadingModel=MSM_Unlit
  usedShadingModels=Unlit
 }
 nodes [
  n1 = ParticleColor()
  n2 = TextureCoordinate() {uTiling=1.5, vTiling=1.5}
  n3 = Subtract(A=TextureSample(UVs=Panner(Coordinate=@n2) {speedX=0.4}) {texture=/BRCosmetics/Gadgets/Assets/VinderTech_GliderChute/Glider_SlickSwish/FX/T_Glider_SlickSwish_noise}.R) {constB=0}
  n4 = TextureCoordinate() {uTiling=0.5, vTiling=0.5}
  n5 = ComponentMask(None=@n4) {g=true}
  n6 = ComponentMask(None=@n4) {r=true}
  n7 = Multiply(A=Multiply(A=Sine(None=@n5), B=Sine(None=@n6))) {constB=1.5}
  n8 = Multiply(
   A=LinearInterpolate(A=@n1.A, B=Multiply(A=@n1.A, B=@n3), Alpha=ScalarParameter() {parameterName=Use_NoiseMask})
   B=Saturate(None=Power(Base=@n7) {constExponent=4})
  )
  n9 = SphereMask(
   A=WorldPosition().XYZ
   B=ParticlePositionWS()
   Radius=ScalarParameter() {defaultValue=65, parameterName=WorldPosition_SphereMaskRadius}
   Hardness=ScalarParameter() {defaultValue=30, parameterName=WorldPosition_SphereMaskHardness}
  ) {attenuationRadius=65, hardnessPercent=10}
  DepthFade(
   Opacity=LinearInterpolate(A=@n8, B=Multiply(A=@n8, B=@n9), Alpha=ScalarParameter() {parameterName=Use_AbsoluteSphereMask})
   FadeDistance=ScalarParameter() {parameterName=DepthFade}
  )
  n10 = ScalarParameter() {defaultValue=16, parameterName=MosaicUV}
  n11 = Multiply(A=TextureCoordinate(), B=@n10)
  n12 = Add(A=Divide(A=Floor(None=@n11), B=@n10), B=Multiply(A=Time(), B=Constant2Vector() {r=-0.3}))
  n13 = Multiply(
   A=LinearInterpolate(
    A=Constant3Vector() {constant=<rgba 1 0.0665 0.6229 1>}
    B=Constant3Vector() {constant=<rgba 0.0448 0.5087 1 1>}
    Alpha=TextureSample(UVs=@n12) {samplerType=SAMPLERTYPE_LinearColor, texture=/Game/Global/Textures/Default/Noise/T_Wrap_SnowStreak_Noise}.RGB
   )
   B=@n1.RGB
  )
  n14 = Add(A=TextureCoordinate() {coordinateIndex=1}, B=Constant2Vector() {g=0.5})
  n15 = Multiply(A=OneMinus(None=ComponentMask(None=@n14) {g=true})) {constB=5}
  n16 = Panner(Coordinate=TextureCoordinate() {uTiling=1.5}) {speedX=0.5}
  n17 = Multiply(
   A=Saturate(None=Power(Base=@n15) {constExponent=5})
   B=Subtract(A=TextureSample(UVs=@n16) {texture=&T_Glider_SlickSwish_noise}.R) {constB=0.3}
  )
  MaterialFunctionCall(Emissive=LinearInterpolate(A=@n13, B=Add(A=@n17, B=@n13), Alpha=ScalarParameter() {parameterName=Use_Line})) {materialFunction=&MF_FX_EmissiveAdjustments}
 ]
}
