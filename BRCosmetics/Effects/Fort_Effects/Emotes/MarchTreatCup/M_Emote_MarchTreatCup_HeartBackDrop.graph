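# Translucent, unlit, two-sided Niagara particle material for the heart backdrop effect of the MarchTreatCup emote.
# Scrolling cell noise indexes a color-curve atlas, tinted by particle color, plus a pink scrolling sweep and optional desaturated rainbow layer.
# Opacity combines particle alpha, a vertical gradient and an optional smoke side mask with per-particle random scrolling, then view-angle and depth fades.
# Scalar blends select the rainbow, side mask or sweep-line-only look.
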Material /BRCosmetics/Effects/Fort_Effects/Emotes/MarchTreatCup/M_Emote_MarchTreatCup_HeartBackDrop
graph {
 properties {
  bAutomaticallySetUsageInEditor=false
  bUsedWithNiagaraMeshParticles=true
  bUsedWithNiagaraSprites=true
  blendMode=BLEND_Translucent
  shadingModel=MSM_Unlit
  twoSided=true
  usedShadingModels=Unlit
 }
 nodes [
  n1 = ParticleColor()
  n2 = ComponentMask(None=TextureCoordinate() {vTiling=0.6}) {g=true}
  n3 = Saturate(None=Power(Base=Multiply(A=@n2) {constB=4}) {constExponent=4.5})
  n4 = VectorParameter() {parameterName=SweepTexture, defaultValue=<rgba 1 3 0 1.2>}
  n5 = Panner(Coordinate=Multiply(A=TextureCoordinate(), B=AppendVector(A=@n4.R, B=@n4.R)), Speed=AppendVector(A=@n4.B, B=@n4.B)) {speedY=1.2}
  n6 = Saturate(None=Power(Base=TextureSampleParameter2D(UVs=@n5) {parameterName=SweepTexture, texture=/Game/PvP/StrategicBuildingActors/Defensive/LaserFence/Assets/T_LaserBeamParticle}.R) {constExponent=4})
  n7 = ScalarParameter() {parameterName=Use_OnceLine}
  n8 = LinearInterpolate(A=Multiply(A=@n1.A, B=@n3), B=Multiply(A=@n1.A, B=Multiply(A=@n6, B=@n3)), Alpha=@n7)
  n9 = ParticleRandom()
  n10 = ScalarParameter() {defaultValue=0.2, parameterName=SideMaskTex_Speed}
  n11 = Panner(Coordinate=Add(A=TextureCoordinate() {vTiling=0.3}, B=@n9), Speed=AppendVector(A=Constant(), B=Add(A=@n9, B=@n10))) {speedY=0.65}
  n12 = ComponentMask(None=TextureCoordinate()) {g=true}
  n13 = Multiply(A=@n12) {constB=3}
  n14 = OneMinus(None=@n12)
  n15 = Saturate(None=Multiply(A=Power(Base=@n13) {constExponent=5}, B=Power(Base=@n14) {constExponent=4}))
  n16 = Sine(None=ComponentMask(None=TextureCoordinate() {uTiling=0.5, vTiling=0.5}) {r=true})
  n17 = Saturate(None=Power(Base=Multiply(A=@n16)) {constExponent=3})
  n18 = Multiply(
   A=@n8
   B=Multiply(
    A=TextureSample(UVs=@n11) {texture=/Game/Global/Textures/Default/Tileable_Patterns/T_Blowing_Smoke_Pattern2_Tile_DISP}.R
    B=Multiply(A=@n15, B=@n17)
   )
  )
  n19 = ScalarParameter() {parameterName=Use_SideMask}
  n20 = Multiply(A=MaterialFunctionCall("FadeSensitivity (optional)"=ScalarParameter() {parameterName=MeshFade}) {materialFunction=&FadeBasedOnViewAngle}.Result) {constB=3}
  DepthFade(
   Opacity=Multiply(A=LinearInterpolate(A=@n8, B=@n18, Alpha=@n19), B=Clamp(None=@n20))
   FadeDistance=ScalarParameter() {defaultValue=10, parameterName=DepthFade}
  )
  n21 = TextureCoordinate() {uTiling=0.1, vTiling=0.5}
  n22 = Subtract(A=TextureSample(UVs=Panner(Coordinate=@n21) {speedY=0.2}) {texture=&T_Tile_RainbowColor}.RGB) {constB=0.2}
  n23 = Constant() {r=0.5}
  n24 = Multiply(A=Constant3Vector() {constant=<rgba 1.5 0.15 0.462 1>}, B=@n6)
  n25 = VectorParameter() {parameterName=BaseTexture, defaultValue=<rgba 0 0.5 -0.5 0>}
  n26 = Multiply(A=TextureCoordinate(), B=AppendVector(A=@n25.R, B=@n25.R))
  n27 = Constant() {r=0.25}
  n28 = TextureSampleParameter2D!(
   UVs=Panner(
    Coordinate=MaterialFunctionCall(UVs=@n26, "Rotation Angle (0-1)"=@n27) {materialFunction=&CustomRotator}.Rotated Values
    Speed=AppendVector(A=@n25.B, B=@n25.B)
   ) {speedX=-0.5}
   Level
   Apply View MipBias
  ) {
   mipValueMode=TMVM_MipLevel
   parameterName=BaseTexture
   samplerType=SAMPLERTYPE_Grayscale
   texture=/BRCosmetics/Weapons/FORT_Melee/Pickaxe_Cube_Ninja_Male/FX/T_CellNoise
  }
  n29 = CurveAtlasRowParameter(CurveTime=LinearInterpolate(Alpha=@n28.R) {constA=0.1}) {
   atlas=/BRCosmetics/Effects/Fort_Effects/Emotes/WarmShade/Materials/CAtlas_Emote_WarmShade
   curve=&/CV_Emote_MarchTreatCup_BackdropColor1
   defaultValue=11
   parameterName=ColorCurve
  }
  n30 = Add(A=@n24, B=Multiply(A=@n29, B=@n1.RGB) {constA=1})
  n31 = Add(A=LinearInterpolate(B=Desaturation(None=@n22, Fraction=@n23)), B=@n30)
  n32 = ScalarParameter() {defaultValue=1, parameterName=Use_ColorTex}
  MaterialFunctionCall(Emissive=LinearInterpolate(A=LinearInterpolate(A=@n31, B=@n30, Alpha=@n32), B=Multiply(A=@n24, B=@n1.RGB), Alpha=@n7)) {materialFunction=&MF_FX_EmissiveAdjustments}
 ]
}
