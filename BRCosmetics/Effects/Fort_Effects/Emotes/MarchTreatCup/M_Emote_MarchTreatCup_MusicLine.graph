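# Translucent unlit Niagara mesh-particle material for the music line effect in the MarchTreatCup emote.
# Opacity blends panned smoke, distorted trail textures or plain particle alpha through parameter switches, then applies edge masks, optional scanlines and sphere mask, and depth fade.
# Color mixes a noise-driven pink-blue gradient or curve-atlas color, tinted by particle color, plus a bright edge line.
# Dynamic parameters drive panning sine-and-mask waves that displace vertices.
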Material /BRCosmetics/Effects/Fort_Effects/Emotes/MarchTreatCup/M_Emote_MarchTreatCup_MusicLine
graph {
 properties {
  bAutomaticallySetUsageInEditor=false
  bUsedWithNiagaraMeshParticles=true
  blendMode=BLEND_Translucent
  shadingModel=MSM_Unlit
  usedShadingModels=Unlit
 }
 nodes [
  n1 = ParticleColor()
  n2 = Panner(Coordinate=Add(A=TextureCoordinate() {vTiling=0.8}, B=Constant2Vector() {g=0.1})) {speedX=-1}
  n3 = Multiply(
   A=@n1.A
   B=Multiply(A=TextureSample(UVs=@n2) {samplerType=SAMPLERTYPE_Grayscale, texture=/Game/Effects/Fort_Effects/Textures/Smoke/T_Smoke_Trail_Soft_NoClamp}.R) {constB=1.5}
  )
  n4 = ScalarParameter() {defaultValue=3, parameterName=WPO_Wave_U}
  n5 = DynamicParameter() {paramNames [WPO_Wave_SpeedY, WPO_Wave_U, WPO_Wave_Intensity, WPO_Music_Height]}
  n6 = Multiply(A=TextureCoordinate(), B=AppendVector(A=Multiply(A=@n4, B=@n5.WPO_Wave_U), B=Constant() {r=1}))
  n7 = Constant() {r=0.25}
  n8 = Constant()
  n9 = Multiply(A=ScalarParameter() {defaultValue=-1, parameterName=WPO_Wave_SpeedY}, B=@n5.WPO_Wave_SpeedY)
  n10 = TextureSample(UVs=Panner(
    Coordinate=MaterialFunctionCall(UVs=@n6, "Rotation Angle (0-1)"=@n7) {materialFunction=&CustomRotator}.Rotated Values
    Speed=AppendVector(A=@n8, B=@n9)
   ) {speedY=-1}) {texture=&/T_Emote_MarchTreatCup_Mask}
  n11 = ComponentMask(None=TextureCoordinate()) {r=true}
  n12 = ScalarParameter() {defaultValue=5, group=WPO, parameterName=WPO A Tiling}
  n13 = Multiply(A=@n5.WPO_Wave_SpeedY, B=ScalarParameter() {defaultValue=1, group=WPO, parameterName=WPO A Speed})
  n14 = Time()
  n15 = Sine(None=Add(A=Multiply(A=@n11, B=@n12) {constB=0.5}, B=Multiply(A=@n13, B=@n14) {constB=3}))
  n16 = Add(A=@n10.R, B=LinearInterpolate(Alpha=@n15) {constA=0.5})
  n17 = Add(A=Add(A=TextureCoordinate(), B=Constant2Vector() {g=-0.2}), B=Multiply(A=@n16) {constB=0.2})
  n18 = Multiply(A=Constant() {r=0.5}, B=@n5.WPO_Wave_SpeedY)
  n19 = Constant() {r=0.5}
  n20 = TextureSample(UVs=Panner(Coordinate=TextureCoordinate() {uTiling=2}, Speed=AppendVector(A=@n18, B=@n19)) {speedX=0.5}) {texture=/BRCosmetics/Gadgets/Assets/VinderTech_GliderChute/Glider_SlickSwish/FX/T_Glider_SlickSwish_noise}
  n21 = Multiply(
   A=@n1.A
   B=Multiply(
    A=TextureSample(UVs=@n17) {texture=/BRCosmetics/Accessories/FORT_Backpacks/Backpack_F_MED_ChimeCurlCorn/FX/T_Backpack_ChimeCurlCorn_Trail}.R
    B=Subtract(A=@n20.R) {constB=0.3}
   )
  )
  n22 = ScalarParameter() {parameterName=Use_LightLine}
  n23 = ComponentMask(None=TextureCoordinate()) {r=true}
  n24 = Power(Base=Multiply(A=Sine(None=@n23) {period=2}) {constB=1.5}) {constExponent=4}
  n25 = ComponentMask(None=TextureCoordinate() {coordinateIndex=1}) {g=true}
  n26 = Power(Base=Multiply(A=OneMinus(None=@n25)) {constB=3})
  n27 = Multiply(
   A=LinearInterpolate(A=LinearInterpolate(A=@n3, B=@n21, Alpha=@n22), B=@n1.A, Alpha=ScalarParameter() {parameterName=Use_TrailMask})
   B=Multiply(A=Saturate(None=@n24), B=Saturate(None=@n26))
  )
  n28 = ScalarParameter() {defaultValue=200, parameterName=Scanline Amount}
  n29 = ScalarParameter() {defaultValue=-0.1, parameterName=ScanlineSize}
  n30 = ScalarParameter() {defaultValue=0.5, parameterName=Scanlines Dark}
  n31 = ScalarParameter() {defaultValue=1, parameterName=Scanlines Light}
  n32 = LinearInterpolate(
   A=@n27
   B=Multiply(
    A=@n27
    B=MaterialFunctionCall!(
     Number of Scanlines=@n28
     Scanline Speed
     "Scanlines Thickness (Bias)"=@n29
     Scanlines Dark Value=@n30
     Scanlines Light Value=@n31
     Block Opacity
     UVs
     Time
     Block Repeat
    ) {materialFunction=/Game/Effects/Fort_Effects/Effects/Atlas/Materials/MF_Atlas_Scanlines_01}.Result
   )
   Alpha=ScalarParameter() {parameterName=Use_Scanline}
  )
  n33 = SphereMask(A=WorldPosition().XYZ, B=ParticlePositionWS(), Radius=ScalarParameter() {defaultValue=65, parameterName=WorldPosition_SphereMaskRadius}) {attenuationRadius=65, hardnessPercent=50}
  DepthFade(
   Opacity=LinearInterpolate(A=@n32, B=Multiply(A=@n32, B=@n33), Alpha=ScalarParameter() {parameterName=Use_AbsoluteSphereMask})
   FadeDistance=ScalarParameter() {defaultValue=5, parameterName=DepthFade}
  )
  n34 = ComponentMask(None=TextureCoordinate()) {g=true}
  n35 = Power(Base=OneMinus(None=Sine(None=@n34) {period=2})) {constExponent=15}
  n36 = Add(A=TextureSample(UVs=TextureCoordinate() {vTiling=1.2}) {texture=&T_Glider_SlickSwish_noise}.R) {constB=0.3}
  n37 = Power(Base=Multiply(A=Multiply(A=@n36, B=@n10.R)) {constB=1.2}) {constExponent=1}
  n38 = MaterialFunctionCall() {materialFunction=/Game/Effects/Niagara/MasterMaterials/Functions/InstanceBounds_0-1_UVW}
  n39 = CurveAtlasRowParameter(CurveTime=Add(A=Divide(A=@n38.B) {constB=3}, B=ScalarParameter() {parameterName=Offset}) {constB=0.25}) {
   atlas=/BRCosmetics/Effects/Fort_Effects/Emotes/WarmShade/Materials/CAtlas_Emote_WarmShade
   curve=&/CV_Emote_MarchTreatCup_Color
   defaultValue=9
   parameterName=Color
  }
  n40 = LinearInterpolate(
   A=LinearInterpolate(A=Constant3Vector() {constant=<rgba 1 0.1583 0.8098 1>}, B=Constant3Vector() {constant=<rgba 0.1536 0.14 1 1>}, Alpha=Saturate(None=@n37))
   B=Desaturation(None=@n39, Fraction=DynamicParameter() {parameterIndex=1, defaultValue=<rgba 0 0 0 1>, paramNames [Des,none,none,none]}.Des)
   Alpha=ScalarParameter() {parameterName=Use_ChangeColorCurve}
  )
  MaterialFunctionCall(Emissive=Add(A=Saturate(None=@n35), B=Multiply(A=@n40, B=@n1.RGB))) {materialFunction=&MF_FX_EmissiveAdjustments}
  n41 = Multiply(
   A=Transform(None=Constant3Vector() {constant=<rgba 0 0 1 0>}) {transformSourceType=TRANSFORMSOURCE_World, transformType=TRANSFORM_Instance}
   B=VertexColor().R
  )
  n42 = Multiply(A=Constant() {r=3}, B=DynamicParameter() {paramNames [WPO_Wave_SpeedY, WPO_Wave_U, WPO_Wave_Intensity, WPO_Music_Height]}.WPO_Music_Height)
  n43 = Time()
  n44 = Time()
  n45 = LinearInterpolate(Alpha=Multiply(A=Sine(None=@n43) {period=0.8}, B=Sine(None=@n44) {period=0.4})) {constA=0.8}
  n46 = Add(A=Multiply(A=@n41, B=Constant() {r=3}) {constB=3}, B=Multiply(A=Multiply(A=@n41, B=@n16), B=Multiply(A=@n42, B=@n45)) {constB=10})
  n47 = Constant3Vector() {constant=<rgba 0 1 0 0>}
  n48 = Multiply(
   A=Multiply(A=Transform(None=@n47) {transformSourceType=TRANSFORMSOURCE_World, transformType=TRANSFORM_Instance}, B=@n15)
   B=Constant() {r=1.2}
  ) {constB=10}
  Reroute(None=Multiply(
    A=Add(A=@n46, B=@n48)
    B=DynamicParameter() {paramNames [WPO_Wave_SpeedY, WPO_Wave_U, WPO_Wave_Intensity, WPO_Music_Height]}.WPO_Wave_Intensity
   ))
 ]
}
