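# Niagara backdrop effect (part 1) for the MarchTreatCup emote, built from two mesh-particle emitters offset in camera depth.
# Backboard bursts meshes for two renderers, animating a secondary color and dynamic material parameters over particle life.
# Ring repeatedly spawns expanding meshes that shift between pink and violet-blue as they fade.
# Both scale opacity by front end, time of day and platform, and adjust vertical offset by LOD distance.
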NiagaraSystem /BRCosmetics/Effects/Fort_Effects/Emotes/MarchTreatCup/NS_Emote_MarchTreatCup_BackDrop_Part1
dependencies {
 usedDataInterfaces [&NiagaraDataInterfaceColorCurve, &NiagaraDataInterfaceCurve, &NiagaraDataInterfacePlatformSet]
 usedDynamicInputs [&RemapRange]
 usedModules [
  &SystemState
  &InterpolateOverTime
  &self:SystemSpawnScript.SystemScriptSource.SystemScriptGraph.NiagaraNodeAssignment_0.SetVariables_#1
 ]
}
emitters {
 Ring {
  emitterUpdateScript [
   &EmitterState(Inactive Response=::NewEnumerator1, Life Cycle Mode=::NewEnumerator1, Loop Behavior=::NewEnumerator1, Loop Duration=2.6, UseLoopDelay=true)
   &SpawnRate(SpawnRate=2.2, Use Spawn Probability=false)
   &SpawnBurst_Instantaneous(Use Spawn Probability=false, UseLoopCountLimit=true)
  ]
  particleSpawnScript [
   &InitializeParticle(
    Color=<rgba 0.7 0.7 0.7 1>
    Mesh Scale=<1.5 1 1>
    Mesh Scale Mode=::NewEnumerator4
    Position Offset=<0 0 130>
    UsePositionOffset=true
    Write Material Random=true
   )
   &CameraOffset(Camera Offset Amount=-100)
   &InitialMeshOrientation(Mesh Orientation Mode=::NewEnumerator2, Rotation=<0 0 0.25>, Use Rotation Vector=true)
  ]
  particleUpdateScript [
   &ParticleState()
   &ScaleMeshSize(Scale Factor=&VectorFromFloat(Value=&Multiply_Float(
      A=&FloatFromCurve(FloatCurve=DataInterface() {propertyValues {Curve {keys [<at 0 0.5>,<at 1 1.05>]}}}, Scale Curve=1.1)
      B=&FN_LOD_Distance_Remap(Start Distance=240, End Distance=600, Minimum Output=1, Maximum Output=1, Clamp Output=true)
     )))
   &ScaleColor(
    Scale Alpha=&FloatFromCurve(
     FloatCurve=DataInterface() {propertyValues {Curve {keys [
         <at 0 0>
         <at 0.2 1>
         {arriveTangent=-3.1699, interpMode=RCIM_Cubic, leaveTangent=-3.1699, tangentMode=RCTM_User, time=0.5, value=1}
         <at 1 0 RCIM_Cubic>
        ]}}}
     Scale Curve=&Multiply_Float(
      A=&Multiply_Float(A=1, B=&FN_FrontEndFloatModifier(Gameplay=1, Front End=0.5))
      B=&Multiply_Float(
       A=&FNTimeOfDayScalar(Night Scale=0.8, Morning Scale=0.9, Evening Scale=0.9)
       B=&FN_PickFloatDependOnPlatform(TargetPlatformActive=1, TargetPlatformNoActive=0.8)
      )
     )
    )
    Scale RGB=&MakeVectorFromLinearColorRGB(Color=&Multiply_LinearColorByFloat(
      Linear Color=&ColorFromCurve(ColorCurve=DataInterface() {propertyValues {
         AlphaCurve {keys []}
         BlueCurve {keys [<at 0.0839 0.4932>,<at 0.2657 0.8856>,<at 0.4764 1>,<at 0.6973 0.8856>,<at 0.9156 0.5722>]}
         GreenCurve {keys [<at 0.0839 0.15>,<at 0.2657 0.2176>,<at 0.4764 0.1186>,<at 0.6973 0.2176>,<at 0.9156 0.1127>]}
         RedCurve {keys [<at 0.0839 1>,<at 0.2657 1>,<at 0.4764 0.1868>,<at 0.6973 1>,<at 0.9156 1>]}
        }})
      Float=1
     ))
    ScaleA=true
    ScaleRGB=true
   )
   &ShapeLocation(
    MaskSpawnGroup=false
    Offset=&Lerp_Vector(
     A=<0 -40 145>
     B=<0 -40 120>
     Alpha=&FN_LOD_Distance_Remap(Start Distance=240, End Distance=600, Minimum Output=0, Maximum Output=1, Clamp Output=true)
    )
    Offset Mode=::NewEnumerator0
    Override Seed=false
    Sphere Radius=0.1
    UseOwnerScale=true
   )
  ]
  rendererClasses [&NiagaraMeshRendererProperties]
  renderers [{rendererClass=&NiagaraMeshRendererProperties, rendererIndex=0}]
 }
 backboard {
  emitterUpdateScript [
   &EmitterState(Inactive Response=::NewEnumerator1, Life Cycle Mode=::NewEnumerator1, Loop Behavior=::NewEnumerator1, Loop Duration=3, UseLoopDelay=true)
   &SpawnBurst_Instantaneous(Use Spawn Probability=false, UseLoopCountLimit=false)
  ]
  particleSpawnScript [
   &InitializeParticle(
    Lifetime=2.8
    Mesh Scale=<1.5 1.5 1.5>
    Mesh Scale Mode=::NewEnumerator4
    Position Offset=<0 0 130>
    Sprite Size=<220 180>
    Sprite Size Mode=::NewEnumerator4
    UsePositionOffset=true
    Write Material Random=true
   )
   &CameraOffset(Camera Offset Amount=-100)
   &InitialMeshOrientation(Mesh Orientation Mode=::NewEnumerator2, Rotation=<0 0 0.25>, Use Rotation Vector=true)
  ]
  particleUpdateScript [
   &ParticleState()
   /SetVariables(Particles.Size_Bottom:Vector3f=&Multiply_VectorByFloat(
     Vector=<1 1.3 1.2>
     Float=&Multiply_Float(
      A=0.85
      B=&FN_LOD_Distance_Remap(Start Distance=240, End Distance=600, Minimum Output=0.85, Maximum Output=1, Clamp Output=true)
     )
    )) {isSetParameters=true}
   /SetVariables(Particles.Color002=&MakeLinearColorFromVectorAndFloat(
     "Vector (RGB)"=&MakeVectorFromLinearColorRGB(Color=&Multiply_LinearColorByFloat(
       Linear Color=&ColorFromCurve(ColorCurve=DataInterface() {propertyValues {
          AlphaCurve {keys []}
          BlueCurve {keys [<at 0.0661 1 RCIM_Cubic>,<at 0.2093 1>,<at 0.3953 1>,<at 0.5573 0.3428>,<at 0.8109 1 RCIM_Cubic>]}
          GreenCurve {keys [<at 0.0661 0.2548 RCIM_Cubic>,<at 0.2093 0>,<at 0.3953 0.6249>,<at 0.5573 0.0781>,<at 0.8109 0.2548 RCIM_Cubic>]}
          RedCurve {keys [<at 0.0661 0.0001 RCIM_Cubic>,<at 0.2093 0.6015>,<at 0.3953 0>,<at 0.5573 1>,<at 0.8109 0.0001 RCIM_Cubic>]}
         }})
       Float=0.5
      ))
     "Float (Alpha)"=&FloatFromCurve(
      FloatCurve=DataInterface() {propertyValues {Curve {keys [
          <at 0 0>
          <at 0.05 1>
          {arriveTangent=-8.4132, interpMode=RCIM_Cubic, leaveTangent=-8.4132, tangentMode=RCTM_User, time=0.9, value=1}
          <at 1 0 RCIM_Cubic>
         ]}}}
      Scale Curve=&Multiply_Float(
       A=&Multiply_Float(A=0.8, B=&FN_FrontEndFloatModifier(Gameplay=1, Front End=0.8))
       B=&Multiply_Float(
        A=&FNTimeOfDayScalar(Night Scale=0.8, Morning Scale=0.9, Evening Scale=0.9)
        B=&FN_PickFloatDependOnPlatform(TargetPlatformActive=1, TargetPlatformNoActive=0.8)
       )
      )
     )
    )) {isSetParameters=true}
   &ScaleMeshSize(Scale Factor=&VectorFromFloat(Value=&Multiply_Float(
      A=0.95
      B=&FN_LOD_Distance_Remap(Start Distance=240, End Distance=600, Minimum Output=1, Maximum Output=1, Clamp Output=true)
     )))
   &ScaleColor(
    Scale Alpha=&FloatFromCurve(
     FloatCurve=DataInterface() {propertyValues {Curve {keys [<at 0 0>,<at 0.05 1>,<at 0.95 1>,<at 1 0>]}}}
     Scale Curve=&Multiply_Float(
      A=&Multiply_Float(A=0.5, B=&FN_FrontEndFloatModifier(Gameplay=1, Front End=0.8))
      B=&Multiply_Float(
       A=&FNTimeOfDayScalar(Night Scale=0.8, Morning Scale=1, Evening Scale=1)
       B=&FN_PickFloatDependOnPlatform(TargetPlatformActive=1, TargetPlatformNoActive=0.8)
      )
     )
    )
    Scale RGB=&MakeVectorFromLinearColorRGB(Color=<rgba 0.6 0.6 0.6 1>)
    ScaleA=true
    ScaleRGB=true
   )
   &DynamicMaterialParameters(
    Index 0 Param 1=&FN_FrontEndFloatModifier(
     Gameplay=&FloatFromCurve(FloatCurve=DataInterface() {propertyValues {Curve {keys [<at 0 0>,<at 0.5 0.1>,<at 1 0.2>]}}})
     Front End=0.3
    )
    Index 0 Param 3=&FloatFromCurve(FloatCurve=DataInterface() {propertyValues {Curve {keys [<at 0 0>,<at 0.4 0>]}}})
    Index 0 Param 4=&FloatFromCurve(FloatCurve=DataInterface() {propertyValues {Curve {keys [
         {arriveTangent=0.8297, time=0, value=0}
         {arriveTangent=1.2, interpMode=RCIM_Cubic, leaveTangent=1.2, time=0.55, value=0.5}
         {arriveTangent=1.5171, interpMode=RCIM_Cubic, leaveTangent=1.5171, tangentMode=RCTM_User, time=1, value=1.2}
        ]}}})
    Module.Index0Float0Write=true
    Module.Index0Float1Write=true
    Module.Index0Float2Write=true
    Module.Index0Float3Write=true
   )
   &ShapeLocation(
    MaskSpawnGroup=false
    Offset=&Lerp_Vector(
     A=<0 -40 145>
     B=<0 -40 120>
     Alpha=&FN_LOD_Distance_Remap(Start Distance=240, End Distance=600, Minimum Output=0, Maximum Output=1, Clamp Output=true)
    )
    Offset Mode=::NewEnumerator0
    Override Seed=false
    Sphere Radius=0.1
    UseOwnerScale=true
   )
  ]
  rendererClasses [&NiagaraMeshRendererProperties]
  renderers [
   {rendererClass=&NiagaraMeshRendererProperties, rendererIndex=0}
   {rendererClass=&NiagaraMeshRendererProperties, rendererIndex=1}
  ]
 }
}
summary {
 systemName=NS_Emote_MarchTreatCup_BackDrop_Part1
 emitters [
  {emitterName=backboard, rendererClasses [&NiagaraMeshRendererProperties]}
  {emitterName=Ring, rendererClasses [&NiagaraMeshRendererProperties]}
 ]
 userVariables [
  User.bIsFrontEnd {type=NiagaraBool}
  User.bIsFrontEndPreview {type=NiagaraBool}
 ]
}
