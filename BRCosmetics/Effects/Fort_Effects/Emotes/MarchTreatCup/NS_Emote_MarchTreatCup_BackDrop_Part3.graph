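# Niagara system for part three of the MarchTreatCup emote backdrop, drawing a glowing heart-themed display at a fixed offset.
# BoxHeart places 28 meshes along a heart outline using partition-indexed position arrays, with scale and alpha blending between audio-amplitude-driven and canned curves via MusicReactivity.
# A multi-mesh backboard cycles colors, while a heart sprite emitter spawns repeated fading pulses.
# Camera-distance remaps adjust layout, with front-end, time-of-day and platform modifiers tuning the look.
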NiagaraSystem /BRCosmetics/Effects/Fort_Effects/Emotes/MarchTreatCup/NS_Emote_MarchTreatCup_BackDrop_Part3
dependencies {
 usedDataInterfaces [
  &NiagaraDataInterfaceCurve
  &NiagaraDataInterfaceColorCurve
  &NiagaraDataInterfacePlatformSet
  &NiagaraDataInterfaceArrayFloat3
 ]
 usedDynamicInputs [
  &MakeCustomLinearColorFromBool
  &RandomBool
  &RemapRange
 ]
 usedModules [
  &SystemState
  &InterpolateOverTime
  &self:SystemSpawnScript.SystemScriptSource.SystemScriptGraph.NiagaraNodeAssignment_1.SetVariables_#1
  &self:SystemSpawnScript.SystemScriptSource.SystemScriptGraph.NiagaraNodeAssignment_0.SetVariables_#2
 ]
}
emitters {
 BoxHeart {
  emitterUpdateScript [
   &EmitterState(Inactive Response=::NewEnumerator1, Life Cycle Mode=::NewEnumerator1, Loop Behavior=::NewEnumerator1, Loop Duration=10, UseLoopDelay=false)
   &SpawnBurst_Instantaneous(Spawn Count=28, Use Spawn Probability=false, UseLoopCountLimit=false)
  ]
  particleSpawnScript [
   &PartitionParticles(Number Of Partitions=28)
   &InitializeParticle(
    Lifetime=6.5
    Mesh Scale Mode=::NewEnumerator1
    Mesh Uniform Scale Max=1.9
    Mesh Uniform Scale Min=1.9
    UsePositionOffset=true
    Write Material Random=true
   )
   /SetVariables(Particles.Randomfloat:NiagaraFloat=&RandomRangeFloat(Minimum=0, Maximum=1)) {isSetParameters=true}
   &CameraOffset(Camera Offset Amount=-100)
   &InitialMeshOrientation(Mesh Orientation Mode=::NewEnumerator2, Rotation=<0 0 0.25>, Use Rotation Vector=true)
  ]
  particleUpdateScript [
   &ParticleState()
   &ScaleMeshSize(Scale Factor=&Multiply_Vector(
     A=<1 1 1>
     B=&Lerp_Vector(
      A=&VectorFromFloat(Value=&Multiply_Float(
        A=&FloatFromCurve(
         FloatCurve=DataInterface() {propertyValues {Curve {keys [
             {interpMode=RCIM_Cubic, tangentMode=RCTM_User, time=0, value=1}
             {arriveTangent=-0.0139, interpMode=RCIM_Cubic, leaveTangent=-0.0139, tangentMode=RCTM_User, time=0.164, value=1.1141}
             {arriveTangent=-0.0019, interpMode=RCIM_Cubic, leaveTangent=-0.0019, time=0.3, value=0.95}
             {interpMode=RCIM_Cubic, tangentMode=RCTM_User, time=0.4918, value=1.1135}
             {arriveTangent=-0.008, interpMode=RCIM_Cubic, leaveTangent=-0.008, time=0.7, value=0.95}
             {arriveTangent=-0.0025, interpMode=RCIM_Cubic, leaveTangent=-0.0025, tangentMode=RCTM_User, time=0.7967, value=1.1111}
             {tangentMode=RCTM_User, time=1, value=1}
            ]}}}
         Scale Curve=&RandomRangeFloat(Minimum=0.7, Maximum=0.8)
        )
        B=&ClampFloat(Float=$NPC.FortniteNPC.EventAudio_Amplitude, Min=0.5, Max=0.9)
       ))
      B=&VectorFromFloat(Value=&Multiply_Float(
        A=0.6
        B=&FloatFromCurve(
         FloatCurve=DataInterface() {propertyValues {Curve {keys [<at 0 0.9>,<at 0.0322 1>,<at 0.0603 0.8699>,<at 0.0902 0.9861>,<at 0.1147 0.8638>,<at 0.1354 0.9894>,<at 0.1534 0.8653>,<at 0.1731 0.9795>,<at 0.1952 0.8739>,<at 0.217 0.994>,<at 0.2404 0.8783>,<at 0.2717 0.9818>,<at 0.2756 0.8743>,<at 0.3119 0.9885>,<at 0.3283 0.8671>,<at 0.3634 1>,<at 0.4066 0.8671>,<at 0.4298 0.9825>,<at 0.4669 0.8671>,<at 0.4915 0.9818>,<at 0.5157 0.8671>,<at 0.5431 0.9988>,<at 0.5565 0.8654>,<at 0.5992 0.9824>,<at 0.6229 0.8671>,<at 0.6541 0.9888>,<at 0.6814 0.8667>,<at 0.7087 0.971>,<at 0.7302 0.8657>,<at 0.7556 0.9799>,<at 0.7829 0.8671>,<at 0.8191 1>,<at 0.8229 0.8729>,<at 0.8618 0.9839>,<at 0.8905 0.8662>,<at 0.9171 0.9977>,<at 0.9353 0.8663>,<at 0.9558 0.9917>,<at 1 0.9>]}}}
         Scale Curve=0.9
        )
       ))
      Alpha=$System.MusicReactivity
     )
    )) {module=ScaleMeshSize001}
   &Color(Color=&LerpLinearColors(StartColor=<rgba 0.6064 0.2145 2 1>, EndColor=<rgba 2 1.3842 0.16 1>, LerpFactor=$Particles.Randomfloat))
   &ScaleColor(
    Color Value To Scale=$Particles.Color
    Scale Alpha=&Lerp_Float(
     A=&Multiply_Float(
      A=1
      B=&FloatFromCurve(
       FloatCurve=DataInterface() {propertyValues {Curve {keys [
           {interpMode=RCIM_Cubic, tangentMode=RCTM_User, time=0, value=0.6}
           {interpMode=RCIM_Cubic, tangentMode=RCTM_User, time=0.5, value=1}
           {tangentMode=RCTM_User, time=1, value=0.6}
          ]}}}
       Scale Curve=&MakeCustomFloatFromBool(Bool=&SetBoolByFloatComparison(A=$NPC.FortniteNPC.EventAudio_Amplitude, B=0), True Float=1, False Float=0)
      )
     )
     B=1
     Alpha=$System.MusicReactivity
    )
    Scale RGB=&Multiply_VectorByFloat(
     Vector=&MakeVectorFromLinearColorRGB(Color=&ColorFromCurve(ColorCurve=DataInterface() {propertyValues {
         AlphaCurve {keys []}
         BlueCurve {keys [
           <at 0 1 RCIM_Cubic>
           <at 0.1479 2.9703>
           {arriveTangent=0.1012, interpMode=RCIM_Cubic, leaveTangent=0.1012, time=0.2842, value=0.2857}
           <at 0.4419 3>
           <at 0.6416 2>
           <at 0.8565 1>
          ]}
         GreenCurve {keys [
           <at 0 1 RCIM_Cubic>
           <at 0.1479 2.2794>
           {arriveTangent=-6.7421, interpMode=RCIM_Cubic, leaveTangent=-6.7421, time=0.2842, value=0.2093}
           <at 0.4419 0.2969>
           <at 0.6416 0.5652>
           <at 0.8565 1>
          ]}
         RedCurve {keys [
           <at 0 1 RCIM_Cubic>
           <at 0.1479 5>
           {arriveTangent=-11.7444, interpMode=RCIM_Cubic, leaveTangent=-11.7444, time=0.2842, value=5}
           <at 0.4419 1.5466>
           <at 0.6416 0.222>
           <at 0.8565 1>
          ]}
        }}))
     Float=&RandomRangeFloat(Minimum=0.5, Maximum=1)
    )
    ScaleA=true
    ScaleRGB=true
   ) {module=ScaleColor001}
   /SetVariables(Particles.Color002=&MakeLinearColorFromVectorAndFloat(
     "Vector (RGB)"=&Multiply_VectorByFloat(
      Vector=&MakeVectorFromLinearColorRGB(Color=&ColorFromCurve(ColorCurve=DataInterface() {propertyValues {
          AlphaCurve {keys []}
          BlueCurve {keys [
            <at 0 1 RCIM_Cubic>
            <at 0.2058 2>
            <at 0.3356 3>
            <at 0.5017 4.2746>
            {arriveTangent=-7.6103, interpMode=RCIM_Cubic, leaveTangent=-7.6103, time=0.6861, value=3}
            <at 0.8678 1.4887>
            <at 0.995 1>
           ]}
          GreenCurve {keys [
            <at 0 1 RCIM_Cubic>
            <at 0.2058 0.202>
            <at 0.3356 0.625>
            <at 0.5017 0.7432>
            {arriveTangent=-0.1295, interpMode=RCIM_Cubic, leaveTangent=-0.1295, time=0.6861, value=0.6186}
            <at 0.8678 0.6957>
            <at 0.995 1>
           ]}
          RedCurve {keys [
            <at 0 1 RCIM_Cubic>
            <at 0.2058 0.3306>
            <at 0.3356 1.723>
            <at 0.5017 0.2232>
            {arriveTangent=7.5855, interpMode=RCIM_Cubic, leaveTangent=7.5855, time=0.6861, value=0.3948}
            <at 0.8678 3>
            <at 0.995 1>
           ]}
         }}))
      Float=&RandomRangeFloat(Minimum=0.3, Maximum=0.5)
     )
     "Float (Alpha)"=&Lerp_Float(
      A=&FloatFromCurve(
       FloatCurve=DataInterface() {propertyValues {Curve {keys [<at 0 0.8 RCIM_Cubic>,<at 0.5 1 RCIM_Cubic>,<at 1 0.8 RCIM_Cubic>]}}}
       Scale Curve=&MakeCustomFloatFromBool(Bool=&SetBoolByFloatComparison(A=$NPC.FortniteNPC.EventAudio_Amplitude, B=0), True Float=1, False Float=0)
      )
      B=1
      Alpha=$System.MusicReactivity
     )
    )) {isSetParameters=true}
   /SetVariables(Particles.ColorEdge:LinearColor=&MakeLinearColorFromVectorAndFloat(
     "Vector (RGB)"=&MakeVectorFromLinearColorRGB(Color=&RandomRangeLinearColor(Minimum=<rgba 0 3.3866 5 1>, Maximum=<rgba 5 0 2.2737 1>))
     "Float (Alpha)"=&Lerp_Float(
      A=&MakeCustomFloatFromBool(Bool=&SetBoolByFloatComparison(A=$NPC.FortniteNPC.EventAudio_Amplitude, B=0), True Float=1, False Float=0)
      B=1
      Alpha=$System.MusicReactivity
     )
    )) {isSetParameters=true}
   &ShapeLocation(
    MaskSpawnGroup=false
    Offset=&AddVectorToPosition(
     Position=&Lerp_Vector(
      A=&SelectVectorFromArray(
       Direct Array Index=$Particles.PartitionParticles.Partition
       Vector Selection Array=DataInterface() {propertyValues {FloatData [
          <0 0 22>
          <10 0 35>
          <24 0 46>
          <36 0 60>
          <48 0 78>
          <58 0 96>
          <69 0 114>
          <75 0 132>
          <75 0 150>
          <66 0 165>
          <52 0 175>
          <38 0 180>
          <24 0 175>
          <14 0 168>
          <0 0 165>
          <-14 0 168>
          <-25 0 175>
          <-38 0 180>
          <-52 0 175>
          <-66 0 165>
          <-75 0 150>
          <-75 0 132>
          <-69 0 114>
          <-58 0 96>
          <-45 0 78>
          <-33 0 60>
          <-22 0 46>
          <-10 0 35>
         ]}}
      )
      B=&SelectVectorFromArray(
       Direct Array Index=$Particles.PartitionParticles.Partition
       Vector Selection Array=DataInterface() {propertyValues {FloatData [
          <0 0 20>
          <20 0 28>
          <35 0 46>
          <49 0 64>
          <66 0 83>
          <83 0 101>
          <94 0 119>
          <99 0 137>
          <94 0 155>
          <83 0 173>
          <67 0 190>
          <49 0 197>
          <31 0 190>
          <15 0 178>
          <0 0 165>
          <-15 0 178>
          <-32 0 190>
          <-49 0 197>
          <-67 0 190>
          <-83 0 173>
          <-94 0 155>
          <-99 0 137>
          <-94 0 119>
          <-83 0 101>
          <-66 0 83>
          <-49 0 64>
          <-35 0 46>
          <-20 0 28>
         ]}}
      )
      Alpha=&FN_LOD_Distance_Remap(Start Distance=240, End Distance=600, Minimum Output=0, Maximum Output=1, Clamp Output=true)
     )
     Vector=&Lerp_Vector(
      A=&RandomRangeVector(Minimum=<0 -37 30>, Maximum=<0 -42 30>)
      B=&RandomRangeVector(Minimum=<0 -37 10>, Maximum=<0 -42 10>)
      Alpha=&FN_LOD_Distance_Remap(Start Distance=240, End Distance=600, Minimum Output=0, Maximum Output=1, Clamp Output=true)
     )
    )
    Offset Mode=::NewEnumerator0
    Override Seed=false
    Sphere Radius=1
    UseOwnerScale=true
   )
  ]
  rendererClasses [&NiagaraMeshRendererProperties]
  renderers [
   {rendererClass=&NiagaraMeshRendererProperties, rendererIndex=0}
   {rendererClass=&NiagaraMeshRendererProperties, rendererIndex=1}
   {rendererClass=&NiagaraMeshRendererProperties, rendererIndex=2}
  ]
 }
 backboard {
  emitterUpdateScript [
   &EmitterState(Inactive Response=::NewEnumerator1, Life Cycle Mode=::NewEnumerator1, Loop Behavior=::NewEnumerator1, Loop Duration=10, UseLoopDelay=true)
   &SpawnBurst_Instantaneous(Use Spawn Probability=false, UseLoopCountLimit=false)
  ]
  particleSpawnScript [
   &InitializeParticle(
    Lifetime=6.5
    Mesh Scale=<0.1 1.5 1.5>
    Mesh Scale Mode=::NewEnumerator4
    Sprite Size=<220 180>
    Sprite Size Mode=::NewEnumerator4
    UsePositionOffset=true
    Write Material Random=true
   )
   &CameraOffset(Camera Offset Amount=-100)
   &InitialMeshOrientation(Mesh Orientation Mode=::NewEnumerator2, Rotation=<0 0 0.25>, Use Rotation Vector=true)
  ]
  particleUpdateScript [
   &ParticleState()
   &ScaleMeshSize(Scale Factor=&VectorFromFloat(Value=&Multiply_Float(
      A=0.95
      B=&FN_LOD_Distance_Remap(Start Distance=240, End Distance=600, Minimum Output=1, Maximum Output=0.88, Clamp Output=true)
     )))
   /SetVariables(Particles.Dyn_Bottom:Vector4f=&MakeVector4(
     X=&FN_FrontEndFloatModifier(
      Gameplay=&FloatFromCurve(FloatCurve=DataInterface() {propertyValues {Curve {keys [<at 0 0>,<at 0.5 0.1>,<at 1 0.2>]}}})
      Front End=0.3
     )
     Y=0
     Z=&FloatFromCurve(FloatCurve=DataInterface() {propertyValues {Curve {keys [<at 0 0>,<at 0.4 0>]}}})
     W=&FloatFromCurve(
      FloatCurve=DataInterface() {propertyValues {Curve {keys [
          {arriveTangent=0.8297, time=0, value=0}
          {arriveTangent=1.2, interpMode=RCIM_Cubic, leaveTangent=1.2, time=0.55, value=0.5}
          {arriveTangent=1.5171, interpMode=RCIM_Cubic, leaveTangent=1.5171, tangentMode=RCTM_User, time=1, value=1.2}
         ]}}}
      Scale Curve=2
     )
    )) {isSetParameters=true}
   /SetVariables(Particles.Size_Bottom:Vector3f=&Multiply_VectorByFloat(
     Vector=<1 1.3 1.2>
     Float=&FN_LOD_Distance_Remap(Start Distance=240, End Distance=600, Minimum Output=1, Maximum Output=0.9, Clamp Output=true)
    )) {isSetParameters=true}
   /SetVariables(Particles.Size_PointLine:Vector3f=&Multiply_VectorByFloat(
     Vector=<1 1.3 1.2>
     Float=&FN_LOD_Distance_Remap(Start Distance=240, End Distance=600, Minimum Output=1, Maximum Output=0.88, Clamp Output=true)
    )) {isSetParameters=true}
   &ScaleColor(
    Scale Alpha=&FloatFromCurve(
     FloatCurve=DataInterface() {propertyValues {Curve {keys [<at 0 0>,<at 0.05 1>,<at 0.95 1>,<at 1 0>]}}}
     Scale Curve=&Multiply_Float(
      A=&Multiply_Float(A=0.5, B=&FN_FrontEndFloatModifier(Gameplay=1, Front End=0.8))
      B=&Multiply_Float(
       A=&FNTimeOfDayScalar(Night Scale=0.8, Morning Scale=1, Evening Scale=1)
       B=&FN_PickFloatDependOnPlatform(TargetPlatformActive=1, TargetPlatformNoActive=0.8)
      )
     )
    )
    Scale RGB=&MakeVectorFromLinearColorRGB(Color=&Multiply_LinearColorByFloat(
      Linear Color=&ColorFromCurve(ColorCurve=DataInterface() {propertyValues {
         AlphaCurve {keys []}
         BlueCurve {keys [<at 0.0661 1 RCIM_Cubic>,<at 0.2093 1>,<at 0.3953 1>,<at 0.5573 0.3428>,<at 0.8109 1 RCIM_Cubic>]}
         GreenCurve {keys [<at 0.0661 0.2548 RCIM_Cubic>,<at 0.2093 0>,<at 0.3953 0.6249>,<at 0.5573 0.0781>,<at 0.8109 0.2548 RCIM_Cubic>]}
         RedCurve {keys [<at 0.0661 0.0001 RCIM_Cubic>,<at 0.2093 0.6015>,<at 0.3953 0>,<at 0.5573 1>,<at 0.8109 0.0001 RCIM_Cubic>]}
        }})
      Float=0.5
     ))
    ScaleA=true
    ScaleRGB=true
   )
   /SetVariables(Particles.ColorLine:LinearColor=&MakeLinearColorFromVectorAndFloat(
     "Vector (RGB)"=&MakeVectorFromLinearColorRGB(Color=<rgba 3 3 3 1>)
     "Float (Alpha)"=&FloatFromCurve(FloatCurve=DataInterface() {propertyValues {Curve {keys [<at 0 0>,<at 0.05 1>,<at 0.95 1>,<at 1 0>]}}})
    )) {isSetParameters=true}
   /SetVariables(Particles.ColorEnergy:LinearColor=&MakeLinearColorFromVectorAndFloat(
     "Vector (RGB)"=&MakeVectorFromLinearColorRGB(Color=&Multiply_LinearColorByFloat(
       Linear Color=&ColorFromCurve(ColorCurve=DataInterface() {propertyValues {
          AlphaCurve {keys []}
          BlueCurve {keys [
            <at 0.0461 1>
            <at 0.0978 1>
            <at 0.1521 1>
            <at 0.2606 1>
            <at 0.3149 1>
            <at 0.3881 1>
            <at 0.4297 0.3975>
            <at 0.532 0.3428>
            {arriveTangent=3.7732, leaveTangent=3.7732, time=0.5724, value=1}
            <at 0.6973 1>
            <at 0.7503 0.3615>
           ]}
          GreenCurve {keys [
            <at 0.0461 0.2548>
            {arriveTangent=-1.5616, leaveTangent=-1.5616, time=0.0978, value=0.2548}
            <at 0.1521 0>
            <at 0.2606 0>
            <at 0.3149 0.1823>
            <at 0.3881 0.1823>
            <at 0.4297 0.0868>
            <at 0.532 0.0781>
            {arriveTangent=1.0146, leaveTangent=1.0146, time=0.5724, value=0.2548}
            <at 0.6973 0.2548>
            <at 0.7503 0>
           ]}
          RedCurve {keys [
            <at 0.0461 0.0001>
            {arriveTangent=3.6854, leaveTangent=3.6854, time=0.0978, value=0.0001}
            <at 0.1521 0.6015>
            <at 0.2606 0.6015>
            <at 0.3149 0.0004>
            <at 0.3881 0.0004>
            <at 0.4297 0.9168>
            <at 0.532 1>
            {arriveTangent=-5.741, leaveTangent=-5.741, time=0.5724, value=0.0001}
            <at 0.6973 0.0001>
            <at 0.7503 1>
           ]}
         }})
       Float=0.5
      ))
     "Float (Alpha)"=&FloatFromCurve(
      FloatCurve=DataInterface() {propertyValues {Curve {keys [<at 0 0>,<at 0.05 1>,<at 0.95 1>,<at 1 0>]}}}
      Scale Curve=&Multiply_Float(
       A=&Multiply_Float(A=0.7, B=&FN_FrontEndFloatModifier(Gameplay=1, Front End=0.5))
       B=&Multiply_Float(
        A=&FNTimeOfDayScalar(Night Scale=0.6, Morning Scale=1, Evening Scale=1)
        B=&FN_PickFloatDependOnPlatform(TargetPlatformActive=1, TargetPlatformNoActive=0.8)
       )
      )
     )
    )) {isSetParameters=true}
   /SetVariables(Particles.Color003:LinearColor=&MakeLinearColorFromVectorAndFloat(
     "Vector (RGB)"=&MakeVectorFromLinearColorRGB(Color=<rgba 0.625 0.625 0.625 1>)
     "Float (Alpha)"=&FloatFromCurve(FloatCurve=DataInterface() {propertyValues {Curve {keys [<at 0 0>,<at 0.05 1>,<at 0.95 1>,<at 1 0>]}}})
    )) {isSetParameters=true}
   /SetVariables(Particles.Color002=&MakeLinearColorFromVectorAndFloat(
     "Vector (RGB)"=&Multiply_VectorByFloat(Vector=&MakeVectorFromLinearColorRGB(Color=<rgba 2 2 2 1>), Float=1)
     "Float (Alpha)"=&FloatFromCurve(FloatCurve=DataInterface() {propertyValues {Curve {keys [<at 0 0>,<at 0.05 1>,<at 0.95 1>,<at 1 0>]}}})
    )) {isSetParameters=true}
   &ShapeLocation(
    MaskSpawnGroup=false
    Offset=&Lerp_Vector(
     A=<0 -40 145>
     B=<0 -40 120>
     Alpha=&FN_LOD_Distance_Remap(Start Distance=240, End Distance=600, Minimum Output=0, Maximum Output=1, Clamp Output=true)
    )
    Offset Mode=::NewEnumerator0
    Override Seed=false
    Sphere Radius=1
    UseOwnerScale=true
   )
  ]
  rendererClasses [&NiagaraMeshRendererProperties]
  renderers [
   {rendererClass=&NiagaraMeshRendererProperties, rendererIndex=0}
   {rendererClass=&NiagaraMeshRendererProperties, rendererIndex=1}
   {rendererClass=&NiagaraMeshRendererProperties, rendererIndex=2}
   {rendererClass=&NiagaraMeshRendererProperties, rendererIndex=3}
   {rendererClass=&NiagaraMeshRendererProperties, rendererIndex=4}
  ]
 }
 heart {
  emitterUpdateScript [
   &EmitterState(Inactive Response=::NewEnumerator1, Life Cycle Mode=::NewEnumerator1, Loop Behavior=::NewEnumerator1, Loop Duration=6, UseLoopDelay=true)
   &SpawnRate(SpawnRate=1.5, Use Spawn Probability=false)
   &SpawnBurst_Instantaneous(Use Spawn Probability=false, UseLoopCountLimit=true)
  ]
  particleSpawnScript [
   &InitializeParticle(
    Lifetime=1.5
    Sprite Size=&Multiply_Vector2D_ByFloat(Float=2.15, Vector2D=<173 185>)
    Sprite Size Mode=::NewEnumerator4
    UsePositionOffset=true
    Write Material Random=true
   )
   &CameraOffset(Camera Offset Amount=-100)
   &SpriteFacingAndAlignment(Sprite Alignment=<0 0 1>, Sprite Facing=<0 1 0>, WriteAlignment=true, WriteFacing=true)
  ]
  particleUpdateScript [
   &ParticleState()
   &ScaleSpriteSize(
    Scale Sprite Size Mode=::NewEnumerator0
    Uniform Scale Factor=&MakeCustomFloatFromBool(Bool=$User.bIsFrontEndPreview, True Float=0.95, False Float=1)
   ) {module=ScaleSpriteSize001}
   &ScaleSpriteSize(
    Uniform Curve Scale=&FN_LOD_Distance_Remap(Start Distance=240, End Distance=600, Minimum Output=1, Maximum Output=0.88, Clamp Output=true)
    Uniform Curve Sprite Scale=DataInterface() {propertyValues {Curve {keys [
        {arriveTangent=0.1453, interpMode=RCIM_Cubic, leaveTangent=0.1453, tangentMode=RCTM_User, time=0, value=0.9}
        {arriveTangent=0.0442, interpMode=RCIM_Cubic, leaveTangent=0.0442, tangentMode=RCTM_User, time=1, value=1}
       ]}}}
   )
   &ScaleColor(
    Scale Alpha=&FloatFromCurve(
     FloatCurve=DataInterface() {propertyValues {Curve {keys [
         <at 0 0 RCIM_Cubic>
         {arriveTangent=-3.801, interpMode=RCIM_Cubic, leaveTangent=-3.801, tangentMode=RCTM_User, time=0.2, value=0.975}
         <at 0.7 0 RCIM_Cubic>
        ]}}}
     Scale Curve=&MakeCustomFloatFromBool(Bool=$User.bIsFrontEndPreview, True Float=0.2, False Float=0.35)
    )
    Scale RGB=&Multiply_Vector(
     A=&MakeVectorFromLinearColorRGB(Color=&ColorFromCurve(ColorCurve=DataInterface() {propertyValues {
         AlphaCurve {keys []}
         BlueCurve {keys [<at 0.4724 1 RCIM_Cubic>]}
         GreenCurve {keys [<at 0.4724 1 RCIM_Cubic>]}
         RedCurve {keys [<at 0.4724 1 RCIM_Cubic>]}
        }}))
     B=&VectorFromFloat(Value=&Multiply_Float(
       A=2
       B=&Multiply_Float(
        A=&FNTimeOfDayScalar(Night Scale=0.6, Morning Scale=0.8, Evening Scale=0.8)
        B=&FN_FrontEndFloatModifier(Gameplay=1, Front End=0.8)
       )
      ))
    )
    ScaleA=true
    ScaleRGB=true
   )
   &DynamicMaterialParameters(
    Index 0 Param 1=&FloatFromCurve(FloatCurve=DataInterface() {propertyValues {Curve {keys [<at 0 -0.2>,<at 1 2.25>]}}})
    Index 0 Param 2=&FloatFromCurve(FloatCurve=DataInterface() {propertyValues {Curve {keys [<at 0 0.6>,<at 1 1>]}}})
    Module.Index0Param0Write=true
    Module.Index0Param1Write=true
    Module.Index0Param2Write=true
    Module.Index0Param3Write=true
   )
   &ShapeLocation(
    MaskSpawnGroup=false
    Offset=&Lerp_Vector(
     A=<0 -40 140>
     B=<0 -40 115>
     Alpha=&FN_LOD_Distance_Remap(Start Distance=240, End Distance=600, Minimum Output=0, Maximum Output=1, Clamp Output=true)
    )
    Offset Mode=::NewEnumerator0
    Override Seed=false
    Sphere Radius=0.1
    UseOwnerScale=true
   )
  ]
  rendererClasses [&NiagaraSpriteRendererProperties]
  renderers [{rendererClass=&NiagaraSpriteRendererProperties}]
 }
}
summary {
 systemName=NS_Emote_MarchTreatCup_BackDrop_Part3
 emitters [
  {emitterName=backboard, rendererClasses [&NiagaraMeshRendererProperties]}
  {emitterName=BoxHeart, rendererClasses [&NiagaraMeshRendererProperties]}
  {emitterName=heart, rendererClasses [&NiagaraSpriteRendererProperties]}
 ]
 userVariables [
  User.bIsFrontEnd {type=NiagaraBool}
  User.bIsFrontEndPreview {type=NiagaraBool}
 ]
}
