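# Niagara ground effect for the MarchTreatCup emote, building a music-reactive stage with lights, glowing lines, hearts and triangles.
# Long-lived Line, PointLine and four rotating stage-light emitters cycle mostly pink-to-blue color curves, with Line adding a real light.
# Event audio amplitude scales hearts and triangles and, with a system-smoothed value, drives MusicLine brightness and material parameters.
# Front-end, time-of-day and LOD distance modifiers tune intensity and size.
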NiagaraSystem /BRCosmetics/Effects/Fort_Effects/Emotes/MarchTreatCup/NS_Emote_MarchTreatCup_Ground
dependencies {
 usedDataInterfaces [
  &NiagaraDataInterfaceArrayFloat3
  &NiagaraDataInterfaceColorCurve
  &NiagaraDataInterfaceCurve
  &NiagaraDataInterfaceArrayFloat
  &NiagaraDataInterfaceArrayColor
  &NiagaraDataInterfaceSpriteRendererInfo
 ]
 usedDynamicInputs [&RemapRange]
 usedModules [
  &SystemState
  &InterpolateOverTime
  &self:SystemSpawnScript.SystemScriptSource.SystemScriptGraph.NiagaraNodeAssignment_1.SetVariables_#1
  &self:SystemSpawnScript.SystemScriptSource.SystemScriptGraph.NiagaraNodeAssignment_0.SetVariables_#2
 ]
}
emitters {
 BoxGlow {
  emitterUpdateScript [
   &EmitterState(Inactive Response=::NewEnumerator1, Life Cycle Mode=::NewEnumerator1, Loop Duration=1, UseLoopDelay=true)
   &SpawnRate(SpawnRate=5, Use Spawn Probability=false)
  ]
  particleSpawnScript [
   &InitializeParticle(
    Color Maximum=<rgba 2 2 2 1>
    Color Minimum=<rgba 1 1 1 1>
    Color Mode=::NewEnumerator2
    Lifetime Max=1.5
    Lifetime Mode=::NewEnumerator1
    Mesh Scale Max=<2.5 2.5 6>
    Mesh Scale Min=<2 2 4>
    Mesh Scale Mode=::NewEnumerator2
    Position Offset=<0 0 10>
    Sprite Rotation Mode=::NewEnumerator2
    Sprite Size Mode=::NewEnumerator1
    Uniform Sprite Size Max=15
    Uniform Sprite Size Min=10
    UsePositionOffset=true
    Write Material Random=true
   )
   &ShapeLocation(
    Handle Radius=15
    Large Radius=30
    MaskSpawnGroup=true
    Non Uniform Scale=<1 1 0.1>
    Offset Mode=::NewEnumerator0
    Override Seed=false
    Shape Primitive=::NewEnumerator3
    Spawn Group:NiagaraInt32=$Particles.SpawnGroup
    UseOwnerScale=true
   )
   &InitialMeshOrientation(
    Mesh Orientation Mode=::NewEnumerator2
    Rotation=&RandomRangeVector(Minimum=<0 0 0>, Maximum=<0 0 1>)
    Use Rotation Vector=true
   )
   &CameraOffset(Camera Offset Amount=10)
  ]
  particleUpdateScript [
   &ParticleState()
   &ScaleColor(
    Scale Alpha=&FloatFromCurve(
     FloatCurve=DataInterface() {propertyValues {Curve {keys [
         <at 0 0 RCIM_Cubic>
         {arriveTangent=0.037, interpMode=RCIM_Cubic, leaveTangent=0.037, tangentMode=RCTM_User, time=0.3, value=1}
         {arriveTangent=-1.7625, interpMode=RCIM_Cubic, leaveTangent=-1.7625, tangentMode=RCTM_User, time=0.6, value=0.4}
         <at 1 0 RCIM_Cubic>
        ]}}}
     Scale Curve=&Multiply_Float(
      A=&Multiply_Float(A=15, B=&FNTimeOfDayScalar(Night Scale=0.8, Morning Scale=1, Evening Scale=1))
      B=&Multiply_Float(A=&FN_FrontEndFloatModifier(Gameplay=1, Front End=0.5), B=1)
     )
    )
    Scale RGB=&MakeVectorFromLinearColorRGB(Color=&ColorFromCurve(ColorCurve=DataInterface() {propertyValues {
        AlphaCurve {keys []}
        BlueCurve {keys [<at 0.2396 0.5237>,<at 0.4446 1>,<at 0.5987 1>,<at 0.8158 0.5213 RCIM_Cubic>]}
        GreenCurve {keys [<at 0.2396 0.15>,<at 0.4446 0.1914>,<at 0.5987 0.0819>,<at 0.8158 0.15 RCIM_Cubic>]}
        RedCurve {keys [<at 0.2396 1>,<at 0.4446 0>,<at 0.5987 0>,<at 0.8158 1 RCIM_Cubic>]}
       }}))
    ScaleA=true
    ScaleRGB=true
   )
   &SolveForcesAndVelocity(Clamp Velocity=false, Limit Acceleration=false)
  ]
  rendererClasses [&NiagaraMeshRendererProperties]
  renderers [{rendererClass=&NiagaraMeshRendererProperties, rendererIndex=0}]
 }
 Heart {
  emitterUpdateScript [
   &EmitterState(Life Cycle Mode=::NewEnumerator1, Loop Duration=1, UseLoopDelay=false)
   &SpawnRate(SpawnRate=3, Use Spawn Probability=false)
   &SpawnRate(SpawnGroup=1, SpawnRate=3, Use Spawn Probability=false) {module=SpawnRate001}
  ]
  particleSpawnScript [
   &FN_SetSpawnGroup(Number of Spawn Groups=2)
   &InitializeParticle(
    Color Maximum=<rgba 0.625 0.0452 0.1312 1.5>
    Color Minimum=<rgba 0.9062 0.085 0.4295 1.5>
    Color Mode=::NewEnumerator2
    Lifetime Max=0.9
    Lifetime Min=0.6
    Lifetime Mode=::NewEnumerator1
    Sprite Size Mode=::NewEnumerator3
    Uniform Sprite Size=50
    UsePositionOffset=false
    Write Material Random=true
   )
   &ShapeLocation(MaskSpawnGroup=true, Offset=<70 0 90>, Offset Mode=::NewEnumerator0, Override Seed=false, Sphere Radius=30, UseOwnerScale=true)
   &ShapeLocation(
    MaskSpawnGroup=true
    Offset=<-70 0 90>
    Offset Mode=::NewEnumerator0
    Override Seed=false
    Spawn Group Mask=1
    Sphere Radius=30
    UseOwnerScale=true
   ) {module=ShapeLocation001}
  ]
  particleUpdateScript [
   &ParticleState()
   &ScaleSpriteSize(
    Scale Sprite Size Mode=::NewEnumerator0
    Uniform Scale Factor=&Multiply_Float(
     A=&Multiply_Float(
      A=&RandomRangeFloat(Minimum=2, Maximum=2.6)
      B=&ClampFloat(Float=$NPC.FortniteNPC.EventAudio_Amplitude, Min=0.55)
     )
     B=&FN_LOD_Distance_Remap(Start Distance=240, End Distance=600, Minimum Output=0.7, Maximum Output=1, Clamp Output=true)
    )
   )
   &SubUVAnimation(
    End Frame Range Override=3
    "Sprite Renderer "=DataInterface() {propertyValues {SpriteRenderer=&self:Heart_0.NiagaraSpriteRendererProperties_0}}
    UseEndFrame=true
    UseStartFrame=true
   )
   &DynamicMaterialParameters(
    Index 0 Param 1=-2
    Index 0 Param 2=&MakeCustomFloatFromBool(Bool=&RandomBool(), True Float=0, False Float=0.35)
    Index 0 Param 3=1
    Index 0 Param 4=&FloatFromCurve(FloatCurve=DataInterface() {propertyValues {Curve {keys [
         <at 0 2>
         {arriveTangent=-12.6958, leaveTangent=-12.6958, time=0.4, value=2}
         {arriveTangent=-1.5184, leaveTangent=-1.5184, time=0.5, value=0.25}
         <at 0.8 0>
        ]}}})
    Module.Index0Float0Write=true
    Module.Index0Float1Write=true
    Module.Index0Float2Write=true
    Module.Index0Float3Write=true
   )
   &ScaleColor(
    Scale RGB=&Multiply_VectorByFloat(
     Vector=<1 1 1>
     Float=&Lerp_Float(A=1, B=1, Alpha=$System.InterpolateOverTime.InterpolateOverTime_Float.Moving Average)
    )
    ScaleA=true
    ScaleRGB=true
   )
   /SetVariables(Particles.Color002=&MakeLinearColorFromVectorAndFloat(
     "Vector (RGB)"=&MakeVectorFromLinearColorRGB(Color=&ColorFromCurve(ColorCurve=DataInterface() {propertyValues {
         AlphaCurve {keys []}
         BlueCurve {keys [{interpMode=RCIM_Cubic, tangentMode=RCTM_User, time=0.5651, value=5}]}
         GreenCurve {keys [{interpMode=RCIM_Cubic, tangentMode=RCTM_User, time=0.5651, value=0.6771}]}
         RedCurve {keys [{interpMode=RCIM_Cubic, tangentMode=RCTM_User, time=0.5651, value=4.4565}]}
        }}))
     "Float (Alpha)"=&FloatFromCurve(FloatCurve=DataInterface() {propertyValues {Curve {keys [
          <at 0 0>
          {arriveTangent=-2.0518, leaveTangent=-2.0518, time=0.05, value=1}
          {arriveTangent=-1.1111, leaveTangent=-1.1111, time=0.9, value=1}
          <at 1 1>
         ]}}})
    )) {isSetParameters=true}
   /SetVariables(Particles.Color003:LinearColor=&MultiplyLinearColor(
     Linear Color=<rgba 1 0.2618 0.61 1>
     Scale Factor=&MakeVector4_FromFloat(Float=&FN_FrontEndFloatModifier(Gameplay=0.9, Front End=0.8))
    )) {isSetParameters=true}
   /SetVariables(Particles.Size003:Vector2f=&Multiply_Vector2D(A=$Particles.SpriteSize, B=&Vector2DFromFloat(Value=1))) {isSetParameters=true}
  ]
  rendererClasses [&NiagaraSpriteRendererProperties]
  renderers [
   {rendererClass=&NiagaraSpriteRendererProperties}
   {rendererClass=&NiagaraSpriteRendererProperties, rendererIndex=1}
   {rendererClass=&NiagaraSpriteRendererProperties, rendererIndex=2}
  ]
 }
 Line {
  emitterUpdateScript [
   &EmitterState(Inactive Response=::NewEnumerator1, Life Cycle Mode=::NewEnumerator1, Loop Duration=15, UseLoopDelay=false)
   &SpawnBurst_Instantaneous(Use Spawn Probability=false, UseLoopCountLimit=false)
  ]
  particleSpawnScript [
   &InitializeParticle(
    Lifetime=16
    Mesh Scale=<0.95 0.95 1.1>
    Mesh Scale Mode=::NewEnumerator4
    Position Offset=<0 0 2>
    Sprite Size Mode=::NewEnumerator3
    Uniform Sprite Size=220
    UsePositionOffset=true
    Write Material Random=true
   )
   &SpriteFacingAndAlignment(Sprite Facing=<0 0 1>, WriteAlignment=true, WriteFacing=true)
  ]
  particleUpdateScript [
   &ParticleState()
   &ScaleMeshSize(Scale Factor=&VectorFromFloat(Value=0.85))
   &ScaleColor(
    Scale Alpha=&Multiply_Float(
     A=2
     B=&FloatFromCurve(FloatCurve=DataInterface() {propertyValues {Curve {keys [<at 0 0>,<at 0.03 1>,<at 0.97 1>,<at 1 0>]}}})
    )
    Scale RGB=&Multiply_VectorByFloat(
     Vector=&MakeVectorFromLinearColorRGB(Color=&ColorFromCurve(ColorCurve=DataInterface() {propertyValues {
         AlphaCurve {keys []}
         BlueCurve {keys [<at 0 0.6692 RCIM_Cubic>,<at 0.1838 1>,<at 0.2499 1>,<at 0.4717 0.6776>,<at 0.6724 1>,<at 0.7684 1>,<at 1 0.6692 RCIM_Cubic>]}
         GreenCurve {keys [<at 0 0.4 RCIM_Cubic>,<at 0.1838 0.6518>,<at 0.2499 0.6518>,<at 0.4717 0.4>,<at 0.6724 0.6518>,<at 0.7684 0.6532>,<at 1 0.4 RCIM_Cubic>]}
         RedCurve {keys [<at 0 1 RCIM_Cubic>,<at 0.1838 0.2618>,<at 0.2499 0.2618>,<at 0.4717 1>,<at 0.6724 0.2618>,<at 0.7684 0.2263>,<at 1 1 RCIM_Cubic>]}
        }}))
     Float=&Multiply_Float(A=1.5, B=&FNTimeOfDayScalar(Night Scale=0.7, Morning Scale=0.7, Evening Scale=0.7))
    )
    ScaleA=true
    ScaleRGB=true
   )
   /SetVariables(Particles.Dyn002:Vector4f=&MakeVector4(X=2, Y=2, Z=0, W=0)) {isSetParameters=true}
   /SetVariables(Particles.Color003:LinearColor=&MakeLinearColorFromVectorAndFloat(
     "Vector (RGB)"=&Multiply_VectorByFloat(
      Vector=&MakeVectorFromLinearColorRGB(Color=&ColorFromCurve(ColorCurve=DataInterface() {propertyValues {
          AlphaCurve {keys []}
          BlueCurve {keys [<at 0 0.614 RCIM_Cubic>,<at 0.1838 1>,<at 0.2499 1>,<at 0.4717 0.5702>,<at 0.6724 1>,<at 0.7684 1>,<at 1 0.614 RCIM_Cubic>]}
          GreenCurve {keys [<at 0 0.3 RCIM_Cubic>,<at 0.1838 0.5117>,<at 0.2499 0.5117>,<at 0.4717 0.2>,<at 0.6724 0.5117>,<at 0.7684 0.5117>,<at 1 0.3 RCIM_Cubic>]}
          RedCurve {keys [<at 0 1 RCIM_Cubic>,<at 0.1838 0.2>,<at 0.2499 0.2>,<at 0.4717 1>,<at 0.6724 0.2>,<at 0.7684 0.2>,<at 1 1 RCIM_Cubic>]}
         }}))
      Float=1
     )
     "Float (Alpha)"=&Multiply_Float(
      A=1
      B=&FloatFromCurve(FloatCurve=DataInterface() {propertyValues {Curve {keys [<at 0 0>,<at 0.03 1>,<at 0.97 1>,<at 1 0>]}}})
     )
    )) {isSetParameters=true}
   &FN_Light_Attributes(
    Color=&ColorFromCurve(ColorCurve=DataInterface() {propertyValues {
       AlphaCurve {keys []}
       BlueCurve {keys [<at 0 0 RCIM_Cubic>,<at 0.0775 0.6266>,<at 0.1544 0.5549>,<at 0.2028 1>,<at 0.3603 1>,<at 0.4049 0.5549>,<at 0.5587 0.5549>,<at 0.6071 1>,<at 0.7534 1>,<at 0.808 0.5549>,<at 0.937 0.5549>,<at 1 0 RCIM_Cubic>]}
       GreenCurve {keys [<at 0 0 RCIM_Cubic>,<at 0.0775 0.349>,<at 0.1544 0.3177>,<at 0.2028 0.3046>,<at 0.3603 0.3046>,<at 0.4049 0.3177>,<at 0.5587 0.3177>,<at 0.6071 0.3046>,<at 0.7534 0.3046>,<at 0.808 0.3177>,<at 0.937 0.3177>,<at 1 0 RCIM_Cubic>]}
       RedCurve {keys [<at 0 0 RCIM_Cubic>,<at 0.0775 1>,<at 0.1544 1>,<at 0.2028 0.1413>,<at 0.3603 0.1413>,<at 0.4049 1>,<at 0.5587 1>,<at 0.6071 0.1413>,<at 0.7534 0.1413>,<at 0.808 1>,<at 0.937 1>,<at 1 0 RCIM_Cubic>]}
      }})
    Create Color Attribute=true
    Create Exponent Attribute=true
    Create Position Attribute=true
    Create Radius Attribute=true
    Exponent=3
    Position=<0 50 100>
    Radius=&Multiply_Float(A=300, B=&FN_FrontEndFloatModifier(Gameplay=1, Front End=0))
    Scale Color=&FloatFromCurve(
     FloatCurve=DataInterface() {propertyValues {Curve {keys [
         {arriveTangent=4.1988, time=0, value=0}
         {arriveTangent=-0.0105, time=0.03, value=1}
         {arriveTangent=-1.0448, leaveTangent=-1.0448, time=0.97, value=1}
         {arriveTangent=-5.4555, leaveTangent=-5.4555, time=1, value=0}
        ]}}}
     Scale Curve=&Multiply_Float(A=&FNTimeOfDayScalar(Night Scale=0.5, Morning Scale=1, Evening Scale=1), B=10)
    )
   )
  ]
  rendererClasses [
   &NiagaraMeshRendererProperties
   &NiagaraSpriteRendererProperties
   &NiagaraLightRendererProperties
  ]
  renderers [
   {rendererClass=&NiagaraMeshRendererProperties, rendererIndex=0}
   {rendererClass=&NiagaraSpriteRendererProperties, rendererIndex=1}
   {rendererClass=&NiagaraLightRendererProperties, rendererIndex=2}
  ]
 }
 Loop_StageLights {
  emitterUpdateScript [
   &EmitterState(Inactive Response=::NewEnumerator1, Life Cycle Mode=::NewEnumerator1, Loop Duration=15, UseLoopDelay=false)
   &SpawnBurst_Instantaneous(Spawn Count=4, Use Spawn Probability=false, UseLoopCountLimit=false)
  ]
  particleSpawnScript [
   &PartitionParticles(Number Of Partitions=4)
   &InitializeParticle(
    Color Mode=::NewEnumerator0
    Lifetime=16
    Mesh Scale=<0.3 0.3 1>
    Mesh Scale Mode=::NewEnumerator4
    Sprite Size Mode=::NewEnumerator3
    UsePositionOffset=false
    Write Material Random=true
   )
   &ShapeLocation(
    MaskSpawnGroup=false
    Offset=&SelectVectorFromArray(
     Direct Array Index=$Particles.PartitionParticles.Partition
     Vector Selection Array=DataInterface() {propertyValues {FloatData [<60 10 12>, <35 -40 12>, <-35 -40 12>, <-60 10 12>]}}
    )
    Offset Mode=::NewEnumerator0
    Override Seed=false
    Sphere Radius=0.1
    UseOwnerScale=true
   )
   &InitialMeshOrientation(
    Mesh Orientation Mode=::NewEnumerator2
    Rotation=&SelectVectorFromArray(
     Direct Array Index=$Particles.PartitionParticles.Partition
     Vector Selection Array=DataInterface() {propertyValues {FloatData [<0.5 0 -0.1>, <0.5 0 0.1>, <0.5 0 -0.1>, <0.5 0 0.1>]}}
    )
    Use Rotation Vector=true
   )
  ]
  particleUpdateScript [
   &ParticleState()
   &UpdateMeshOrientation(
    Rotation Coordinate Space=::NewEnumerator2
    Rotation Vector=&SelectVectorFromArray(
     Direct Array Index=$Particles.PartitionParticles.Partition
     Vector Selection Array=DataInterface() {propertyValues {FloatData [<0 0 0.5>, <0 0 0.5>, <0 0 0.5>, <0 0 0.5>]}}
    )
   )
   &DynamicMaterialParameters(
    Index 0 Param 1=0.15
    Module.Index0Param0Write=true
    Module.Index0Param1Write=true
    Module.Index0Param2Write=true
    Module.Index0Param3Write=true
   )
   /SetVariables(
    Particles.RingPosition:NiagaraPosition=&AddVectorToPosition(Position=$Particles.Position, Vector=<0 0 3>)
    Particles.RingScale:Vector3f=&VectorFromFloat(Value=0.2)
   ) {isSetParameters=true}
   /SetVariables(Particles.DiskScale:Vector3f=&Multiply_VectorByFloat(Vector=<1 1 0.5>, Float=0.92)) {isSetParameters=true}
   /SetVariables(
    Particles.GlowColor=&MakeLinearColorFromVectorAndFloat(
     "Vector (RGB)"=&Multiply_Vector(
      A=&MakeVectorFromLinearColorRGB(Color=&ColorFromCurve(ColorCurve=DataInterface() {propertyValues {
          AlphaCurve {keys []}
          BlueCurve {keys [<at 0.1028 0.8471>,<at 0.1302 0.8879>,<at 0.2711 0.8879>,<at 0.296 0.8471>,<at 0.5191 0.8471>,<at 0.5465 0.8879>,<at 0.7559 0.8879>,<at 0.7808 0.8471>]}
          GreenCurve {keys [<at 0.1028 0.2>,<at 0.1302 0.5241>,<at 0.2711 0.5241>,<at 0.296 0.2>,<at 0.5191 0.2>,<at 0.5465 0.5241>,<at 0.7559 0.5241>,<at 0.7808 0.2>]}
          RedCurve {keys [<at 0.1028 1>,<at 0.1302 0.1776>,<at 0.2711 0.1776>,<at 0.296 1>,<at 0.5191 1>,<at 0.5465 0.1776>,<at 0.7559 0.1776>,<at 0.7808 1>]}
         }}))
      B=&VectorFromFloat(Value=1)
     )
     "Float (Alpha)"=0.5
    )
    Particles.GlowScale:Vector2f=&Vector2DFromFloat(Value=65)
   ) {isSetParameters=true}
   &ScaleColor(
    Scale Alpha=&FloatFromCurve(FloatCurve=DataInterface() {propertyValues {Curve {keys [<at 0 1>,<at 0.1 1>,<at 0.7 1>,<at 1 1>]}}})
    Scale RGB=&Multiply_Vector(
     A=&MakeVectorFromLinearColorRGB(Color=&ColorFromCurve(
       ColorCurve=DataInterface() {propertyValues {
         AlphaCurve {keys []}
         BlueCurve {keys [<at 0.1494 0.8509>,<at 0.4063 0.8879>,<at 0.6546 0.8879>,<at 0.8997 0.8879 RCIM_Cubic>]}
         GreenCurve {keys [<at 0.1494 0.22>,<at 0.4063 0.5606>,<at 0.6546 0>,<at 0.8997 0.6226 RCIM_Cubic>]}
         RedCurve {keys [<at 0.1494 1>,<at 0.4063 0>,<at 0.6546 0.4788>,<at 0.8997 0.0279 RCIM_Cubic>]}
        }}
       CurveIndex=&Multiply_Float(A=&Sine(Normalized Angle=$Emitter.Age, Period=2, Scale=1, Bias=0), B=$System.AudioScale)
      ))
     B=&VectorFromFloat(Value=&Multiply_Float(
       A=&Multiply_Float(A=1, B=&FN_FrontEndFloatModifier(Gameplay=1, Front End=0.6))
       B=&FNTimeOfDayScalar(Night Scale=0.6, Morning Scale=1, Evening Scale=1)
      ))
    )
    ScaleA=true
    ScaleRGB=true
   ) {enabled=false, module=ScaleColor001}
   &ScaleColor(
    Scale Alpha=&FloatFromCurve(FloatCurve=DataInterface() {propertyValues {Curve {keys [<at 0 0>,<at 0.03 1>,<at 0.97 1>,<at 1 0>]}}}, Scale Curve=0.8)
    Scale RGB=&MakeVectorFromLinearColorRGB(Color=&ColorFromCurve(ColorCurve=DataInterface() {propertyValues {
        AlphaCurve {keys []}
        BlueCurve {keys [<at 0.1028 0.8376>,<at 0.1302 0.8879>,<at 0.2711 0.8879>,<at 0.296 0.8376>,<at 0.5191 0.8376>,<at 0.5465 0.8879>,<at 0.7559 0.8879>,<at 0.7808 0.8376>]}
        GreenCurve {keys [<at 0.1028 0.15>,<at 0.1302 0.4332>,<at 0.2711 0.4332>,<at 0.296 0.15>,<at 0.5191 0.15>,<at 0.5465 0.4332>,<at 0.7559 0.4332>,<at 0.7808 0.15>]}
        RedCurve {keys [<at 0.1028 1>,<at 0.1302 0>,<at 0.2711 0>,<at 0.296 1>,<at 0.5191 1>,<at 0.5465 0>,<at 0.7559 0>,<at 0.7808 1>]}
       }}))
    ScaleA=true
    ScaleRGB=true
   )
  ]
  rendererClasses [&NiagaraMeshRendererProperties, &NiagaraSpriteRendererProperties]
  renderers [
   {rendererClass=&NiagaraMeshRendererProperties, rendererIndex=0}
   {rendererClass=&NiagaraMeshRendererProperties, rendererIndex=1}
   {rendererClass=&NiagaraMeshRendererProperties, rendererIndex=2}
   {rendererClass=&NiagaraSpriteRendererProperties, rendererIndex=3}
  ]
 }
 MusicLine {
  emitterUpdateScript [
   &EmitterState(
    Inactive Response=::NewEnumerator1
    Life Cycle Mode=::NewEnumerator1
    Loop Behavior=::NewEnumerator1
    Loop Duration Mode=::NewEnumerator1
    UseLoopDelay=false
   )
   &SpawnBurst_Instantaneous(Spawn Count=4, Use Spawn Probability=false, UseLoopCountLimit=false)
  ]
  particleSpawnScript [
   &PartitionParticles(Number Of Partitions=4)
   &InitializeParticle(
    Color Mode=::NewEnumerator0
    Lifetime=5
    Mesh Scale=<0.75 1 1>
    Mesh Scale Mode=::NewEnumerator4
    Position Offset=&AddVectorToPosition(
     Position=&SelectVectorFromArray(
      Direct Array Index=$Particles.PartitionParticles.Partition
      Vector Selection Array=DataInterface() {propertyValues {FloatData [<0 0 0>, <6 15 0>, <-5 35 0>, <0 55 0>]}}
     )
     Vector=<0 -27 0>
    )
    UsePositionOffset=true
    Write Material Random=true
   )
   /SetVariables(Particles.Randomfloat:NiagaraFloat=&RandomRangeFloat(Minimum=0, Maximum=1)) {isSetParameters=true}
  ]
  particleUpdateScript [
   &ParticleState(Kill Particles When Lifetime Has Elapsed=false, Let Infinitely Lived Particles Die When Emitter Deactivates=true)
   &ScaleMeshSize(Scale Factor=&VectorFromFloat(Value=0.85))
   &DynamicMaterialParameters(
    Index 0 Param 1=&SelectFloatFromArray(
     Direct Array Index=$Particles.PartitionParticles.Partition
     Float Selection Array=DataInterface() {propertyValues {FloatData [1.5,-1.2,1.3,-1,1.5,-1.5]}}
    )
    Index 0 Param 2=&SelectFloatFromArray(
     Direct Array Index=$Particles.PartitionParticles.Partition
     Float Selection Array=DataInterface() {propertyValues {FloatData [1,1.5,1,1.5,2,1]}}
    )
    Index 0 Param 3=&Lerp_Float(
     A=1.3
     B=&RandomRangeFloat(Minimum=1, Maximum=0.5)
     Alpha=&FN_LOD_Distance_Remap(Start Distance=240, End Distance=600, Minimum Output=0, Maximum Output=1, Clamp Output=true)
    )
    Index 0 Param 4=&Lerp_Float(
     A=&Lerp_Float(
      A=1
      B=&RandomRangeFloat(Minimum=3, Maximum=4)
      Alpha=$System.InterpolateOverTime.InterpolateOverTime_Float.Moving Average
     )
     B=2
     Alpha=$System.MusicReactivity
    )
    Module.Index0Float0Write=true
    Module.Index0Float1Write=true
    Module.Index0Float2Write=true
    Module.Index0Float3Write=true
   )
   /SetVariables(Particles.ColorLine:LinearColor=&MakeLinearColorFromVectorAndFloat(
     "Vector (RGB)"=&Multiply_Vector(
      A=&MakeVectorFromLinearColorRGB(Color=<rgba 0 4 2.1256 1>)
      B=&VectorFromFloat(Value=&Lerp_Float(
        A=2
        B=&MakeCustomFloatFromBool(Bool=$User.bIsFrontEndPreview, True Float=5, False Float=8)
        Alpha=&Power(
         a=&Multiply_Float(A=$System.InterpolateOverTime.InterpolateOverTime_Float.Moving Average, B=$NPC.FortniteNPC.EventAudio_Amplitude)
         b=1
        )
       ))
     )
     "Float (Alpha)"=&Lerp_Float(
      A=&Lerp_Float(A=0, B=1, Alpha=$System.InterpolateOverTime.InterpolateOverTime_Float.Moving Average)
      B=1
      Alpha=$System.MusicReactivity
     )
    )) {isSetParameters=true}
   &ScaleColor(
    Scale RGB=&Multiply_VectorByFloat(
     Vector=&MakeVectorFromLinearColorRGB(Color=&MakeCustomLinearColorFromBool(Bool=&RandomBool(), True Color=<rgba 1 1 1 1>, False Color=<rgba 0.5 0.6069 1 1>))
     Float=&FNTimeOfDayScalar(Night Scale=0.9, Morning Scale=1, Evening Scale=0.9)
    )
    ScaleA=true
    ScaleRGB=true
   )
  ]
  rendererClasses [&NiagaraMeshRendererProperties]
  renderers [
   {rendererClass=&NiagaraMeshRendererProperties, rendererIndex=0}
   {rendererClass=&NiagaraMeshRendererProperties, rendererIndex=1}
  ]
 }
 PointLine {
  emitterUpdateScript [
   &EmitterState(Inactive Response=::NewEnumerator1, Life Cycle Mode=::NewEnumerator1, Loop Duration=15, UseLoopDelay=false)
   &SpawnBurst_Instantaneous(Use Spawn Probability=false, UseLoopCountLimit=false)
  ]
  particleSpawnScript [&InitializeParticle(
    Color Mode=::NewEnumerator0
    Lifetime=16
    Mesh Scale=<0.95 1.1 1>
    Mesh Scale Mode=::NewEnumerator4
    Position Offset=<0 0 5>
    UsePositionOffset=true
    Write Material Random=true
   )]
  particleUpdateScript [
   &ParticleState()
   &ScaleMeshSize(Scale Factor=&VectorFromFloat(Value=0.8))
   &ScaleColor(
    Scale Alpha=&Multiply_Float(
     A=1
     B=&FloatFromCurve(FloatCurve=DataInterface() {propertyValues {Curve {keys [<at 0 0>,<at 0.03 1>,<at 0.97 1>,<at 1 0>]}}})
    )
    Scale RGB=&Multiply_VectorByFloat(
     Vector=&MakeVectorFromLinearColorRGB(Color=&ColorFromCurve(ColorCurve=DataInterface() {propertyValues {
         AlphaCurve {keys []}
         BlueCurve {keys [<at 0 0.4762 RCIM_Cubic>,<at 0.1838 1>,<at 0.2499 1>,<at 0.4717 0.4896>,<at 0.6724 1>,<at 0.7684 1>,<at 1 0.4762 RCIM_Cubic>]}
         GreenCurve {keys [<at 0 0.05 RCIM_Cubic>,<at 0.1838 0.5519>,<at 0.2499 0.5519>,<at 0.4717 0.05>,<at 0.6724 0.5519>,<at 0.7684 0.5741>,<at 1 0.05 RCIM_Cubic>]}
         RedCurve {keys [<at 0 1 RCIM_Cubic>,<at 0.1838 0.05>,<at 0.2499 0.05>,<at 0.4717 1>,<at 0.6724 0.05>,<at 0.7684 0.05>,<at 1 1 RCIM_Cubic>]}
        }}))
     Float=&MakeCustomFloatFromBool(Bool=$User.bIsFrontEndPreview, True Float=2.5, False Float=3.5)
    )
    ScaleA=true
    ScaleRGB=true
   )
  ]
  rendererClasses [&NiagaraMeshRendererProperties]
  renderers [
   {rendererClass=&NiagaraMeshRendererProperties, rendererIndex=0}
   {rendererClass=&NiagaraMeshRendererProperties, rendererIndex=1}
  ]
 }
 Triangle {
  emitterUpdateScript [
   &EmitterState(Inactive Response=::NewEnumerator1, Life Cycle Mode=::NewEnumerator1, Loop Duration=4, UseLoopDelay=false)
   &SpawnRate(SpawnRate=20, Use Spawn Probability=false)
   &SpawnRate(SpawnGroup=1, SpawnRate=20, Use Spawn Probability=false) {module=SpawnRate001}
  ]
  particleSpawnScript [
   &FN_SetSpawnGroup(Number of Spawn Groups=2)
   &InitializeParticle(
    Color=&SelectLinearColorFromArray(
     Array Sampling Mode=::NewEnumerator1
     Direct Array Index=&RandomRangeInt(Minimum=0, Maximum=2)
     Color Selection Array=DataInterface() {propertyValues {ColorData [<rgba 0.7317 0.2341 2 1>, <rgba 0.0791 0.7623 2 1>, <rgba 2 0.4375 1.4475 1>]}}
     Random Seed=?
    )
    Lifetime Max=1
    Lifetime Min=0.2
    Lifetime Mode=::NewEnumerator1
    Mass Max=1.5
    Mass Mode=::NewEnumerator3
    Mesh Scale Mode=::NewEnumerator1
    Mesh Uniform Scale Max=0.45
    Mesh Uniform Scale Min=0.15
    Sprite Rotation Angle=&RandomRangeFloat(Minimum=-0.15, Maximum=0.15)
    Sprite Rotation Mode=::NewEnumerator1
    Sprite Size Mode=::NewEnumerator1
    Uniform Sprite Size Max=3.5
    Uniform Sprite Size Min=3
    UsePositionOffset=false
    Write Material Random=true
   )
   &ShapeLocation(
    Cylinder Height=200
    Cylinder Radius=60
    MaskSpawnGroup=true
    Offset=<50 0 100>
    Offset Mode=::NewEnumerator0
    Override Seed=false
    Shape Primitive=::NewEnumerator1
    Surface Only=false
    UseOwnerScale=true
   )
   &ShapeLocation(
    Cylinder Height=200
    Cylinder Radius=60
    MaskSpawnGroup=true
    Offset=<-50 0 100>
    Offset Mode=::NewEnumerator0
    Override Seed=false
    Shape Primitive=::NewEnumerator1
    Spawn Group Mask=1
    Surface Only=false
    UseOwnerScale=true
   ) {module=ShapeLocation001}
   &InitialMeshOrientation(
    Mesh Orientation Mode=::NewEnumerator2
    Rotation=&RandomRangeVector(Minimum=<0 0 0>, Maximum=<1 1 1>)
    Use Rotation Vector=true
   )
   &AddVelocity(Velocity=&RandomRangeVector(Minimum=<0 0 10>, Maximum=<0 0 20>))
   &ApplyInitialForces()
   /SetVariables(Particles.VisibilityTag=&RandomRangeInt(Minimum=0, Maximum=1)) {isSetParameters=true}
  ]
  particleUpdateScript [
   &ParticleState()
   &SpriteRotationRate(Rotation Rate=&Multiply_Float(
     A=&RandomRangeFloat(Minimum=-20, Maximum=20)
     B=&FloatFromCurve(FloatCurve=DataInterface() {propertyValues {Curve {keys [<at 0 1>,<at 1 0.45>]}}}, Scale Curve=10)
    ))
   &ScaleSpriteSize(
    Uniform Curve Scale=&RandomRangeFloat(Minimum=1.2, Maximum=1.5)
    Uniform Curve Sprite Scale=DataInterface() {propertyValues {Curve {keys [{tangentMode=RCTM_User, time=0, value=0.4}, <at 0.1 1.2>, <at 0.2 1>, {tangentMode=RCTM_User, time=1, value=1}]}}}
   )
   /SetVariables(Particles.Size003:Vector2f=&Multiply_Vector2D(
     A=&Vector2DFromFloat(Value=20)
     B=&Vector2DFromFloat(Value=&FloatFromCurve(FloatCurve=DataInterface() {propertyValues {Curve {keys [<at 0 0.6>,<at 0.2 1>,<at 0.8 1>,<at 1 0.6>]}}}))
    )) {isSetParameters=true}
   /SetVariables(Particles.Color003:LinearColor=&MakeLinearColorFromVectorAndFloat(
     "Vector (RGB)"=&MakeVectorFromLinearColorRGB(Color=$Particles.Color)
     "Float (Alpha)"=&FloatFromCurve(
      FloatCurve=DataInterface() {propertyValues {Curve {keys [<at 0 0 RCIM_Cubic>,<at 0.2 1 RCIM_Cubic>,<at 1 0 RCIM_Cubic>]}}}
      Scale Curve=&Multiply_Float(
       A=0.12
       B=&Multiply_Float(A=&FN_FrontEndFloatModifier(Gameplay=1, Front End=0.8), B=1)
      )
     )
    )) {isSetParameters=true}
   /SetVariables(Particles.Dyn002:Vector4f=&MakeVector4(X=2, Y=1, Z=0, W=0)) {isSetParameters=true}
   &UpdateMeshOrientation(
    Rotation Rate=&RandomRangeFloat(Minimum=0.5, Maximum=1.2)
    Rotation Vector=&RandomRangeVector(Minimum=<0.2 0.2 0.2>, Maximum=<0.6 0.6 0.6>)
   )
   &ScaleMeshSize(Scale Factor=&VectorFromFloat(Value=&FloatFromCurve(
      FloatCurve=DataInterface() {propertyValues {Curve {keys [
          {interpMode=RCIM_Cubic, tangentMode=RCTM_User, time=0, value=0}
          {interpMode=RCIM_Cubic, tangentMode=RCTM_User, time=0.2, value=1}
          {arriveTangent=-0.0093, interpMode=RCIM_Cubic, leaveTangent=-0.0093, tangentMode=RCTM_User, time=0.6, value=1}
          {tangentMode=RCTM_User, time=1, value=0}
         ]}}}
      Scale Curve=&Multiply_Float(
       A=&RandomRangeFloat(Minimum=1.5, Maximum=2)
       B=&ClampFloat(Float=$NPC.FortniteNPC.EventAudio_Amplitude, Min=0.4)
      )
     )))
   &DynamicMaterialParameters(
    Index 0 Param 1=&Add_Float(
     A=20
     B=&Lerp_Float(
      A=&Multiply_Float(A=$NPC.FortniteNPC.EventAudio_Amplitude, B=$NPC.FortniteNPC.EventAudio_Amplitude)
      B=1
      Alpha=$System.MusicReactivePreview
     )
    )
    Index 0 Param 2=5
    Index 0 Param 3=&RandomRangeFloat(Minimum=0, Maximum=0.2)
    Index 0 Param 4=&RandomRangeFloat(Minimum=0, Maximum=0.2)
    Module.Index0Param0Write=true
    Module.Index0Param1Write=true
    Module.Index0Param2Write=true
    Module.Index0Param3Write=true
   )
   &ScaleColor(
    Scale Alpha=&RandomRangeFloat(Minimum=1, Maximum=3)
    Scale RGB=&Multiply_Vector(
     A=&MakeVectorFromLinearColorRGB(Color=<rgba 1 1 1 1>)
     B=&VectorFromFloat(Value=&FloatFromCurve(FloatCurve=DataInterface() {propertyValues {Curve {keys [<at 0 1>,<at 0.1157 0.0721>,<at 0.228 0.9759>,<at 0.3274 0.0692>,<at 0.4199 0.8624>,<at 0.5141 0.3216>,<at 0.617 0.9504>,<at 0.7287 0.0762>,<at 0.8508 1.0029>,<at 1 0>]}}}))
    )
    ScaleA=true
    ScaleRGB=true
   )
   &SolveForcesAndVelocity(Clamp Velocity=false, Limit Acceleration=false)
  ]
  rendererClasses [&NiagaraMeshRendererProperties, &NiagaraSpriteRendererProperties]
  renderers [
   {rendererClass=&NiagaraMeshRendererProperties, rendererIndex=0}
   {rendererClass=&NiagaraSpriteRendererProperties, rendererIndex=1}
  ]
 }
}
summary {
 systemName=NS_Emote_MarchTreatCup_Ground
 emitters [
  {
   emitterName=Loop_StageLights
   rendererClasses [&NiagaraMeshRendererProperties, &NiagaraSpriteRendererProperties]
  }
  {
   emitterName=Line
   rendererClasses [
    &NiagaraMeshRendererProperties
    &NiagaraSpriteRendererProperties
    &NiagaraLightRendererProperties
   ]
  }
  {emitterName=PointLine, rendererClasses [&NiagaraMeshRendererProperties]}
  {emitterName=BoxGlow, rendererClasses [&NiagaraMeshRendererProperties]}
  {emitterName=MusicLine, rendererClasses [&NiagaraMeshRendererProperties]}
  {emitterName=Triangle, rendererClasses [&NiagaraMeshRendererProperties, &NiagaraSpriteRendererProperties]}
  {emitterName=Heart, rendererClasses [&NiagaraSpriteRendererProperties]}
 ]
 userVariables [
  User.bIsFrontEnd {type=NiagaraBool}
  User.bIsFrontEndPreview {type=NiagaraBool}
 ]
}
